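# Niagara effect for the Thrive Female glider that forms and animates a mesh cloud, with variations for front-end display.
# Two near-identical emitters burst-spawn meshes, differing in front-end color brightness.
# Age-driven curves animate position, tilt and dynamic material parameters, including two world-space feet points and a timer sped by an interpolated system value.
# User.Deployed adds a large offset to curve time, jumping to the settled pose, and affects loop delay.
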NiagaraSystem /BRCosmetics/Gadgets/Assets/VinderTech_GliderChute/Glider_Thrive_Female/FX/FX_Thrive
dependencies {
 usedDataInterfaces [&NiagaraDataInterfaceCurve, &NiagaraDataInterfaceColorCurve]
 usedDynamicInputs [&Boolean_NotOperation]
 usedModules [
  &self:SystemSpawnScript.SystemScriptSource.SystemScriptGraph.NiagaraNodeAssignment_0.SetVariables_#1
  &SystemState
  &InterpolateOverTime
 ]
}
emitters {
 Cloud {
  emitterUpdateScript [
   &EmitterState(
    Inactive Response=::NewEnumerator1
    Life Cycle Mode=::NewEnumerator1
    Loop Behavior=::NewEnumerator1
    Loop Delay=&MakeCustomFloatFromBool(
     Bool=&SetBoolByIntComparison(A=$User.Deployed, B=0, Comparison Type=::NewEnumerator0)
     True Float=0
     False Float=1.1
    )
    Loop Duration Mode=::NewEnumerator1
    UseLoopDelay=true
   )
   &SpawnBurst_Instantaneous(Use Spawn Probability=false, UseLoopCountLimit=false)
  ]
  particleSpawnScript [
   &InitializeParticle(
    Color Mode=::NewEnumerator0
    Lifetime=5
    Mesh Scale=&Multiply_VectorByFloat(Vector=<0.1 0.1 0.1>, Float=1.3)
    Mesh Scale Mode=::NewEnumerator4
    Position Offset=<0 -15 -8>
    Position Offset Coordinate Space=::Local
    UsePositionOffset=true
    Write Material Random=true
   )
   &InitialMeshOrientation(Mesh Axis To Orient=<0 1 0>, Mesh Orientation Mode=::NewEnumerator2, Orientation Coordinate Space=::Local, Use Rotation Vector=false)
  ]
  particleUpdateScript [
   &ParticleState(Kill Particles When Lifetime Has Elapsed=false, Let Infinitely Lived Particles Die When Emitter Deactivates=true)
   &ScaleMeshSize(Scale Factor=&VectorFromFloat(Value=&MakeCustomFloatFromBool(Bool=$User.IsFrontEnd, True Float=0.9, False Float=1)))
   /SetVariables(Particles.FeetPos:Vector3f=/Niagara/DynamicInputs/Transforms/TransformPosition(
     Position=&Add_Vector(A=Linked() {linkedVariable {name=Particles.Position, type=NiagaraPosition}}, B=<2.5 -15 20>)
     Source Space=::Local
     Destination Space=::World
    )) {isSetParameters=true, module=SetVariables_#2}
   /SetVariables(Particles.Feet2Pos:Vector3f=&TransformPosition(
     Position=&Add_Vector(A=Linked() {linkedVariable {name=Particles.Position, type=NiagaraPosition}}, B=<2.5 45 10>)
     Source Space=::Local
     Destination Space=::World
    )) {isSetParameters=true, module=SetVariables_#3}
   &FN_Timer(Increment Multiplier:NiagaraFloat=&FloatFromCurve(
     FloatCurve=DataInterface() {propertyValues {Curve {keys [{interpMode=RCIM_Cubic, tangentMode=RCTM_User, time=0, value=5.5}, {interpMode=RCIM_Cubic, tangentMode=RCTM_User, time=2, value=1}]}}}
     CurveIndex=&Add_Float(A=$Particles.Age, B=&Multiply_FloatByInt(Float=100, Integer=$User.Deployed))
     Scale Curve=&Lerp_Float(
      A=1
      B=&FloatFromCurve(
       FloatCurve=DataInterface() {propertyValues {Curve {keys [<at 0 1>,<at 2 2.35>]}}}
       CurveIndex=&Add_Float(A=$Particles.Age, B=&Multiply_FloatByInt(Float=100, Integer=$User.Deployed))
      )
      Alpha=&ClampFloat(Float=$System.InterpolateOverTime.InterpolateOverTime_Float.Moving Average)
     )
    ))
   &DynamicMaterialParameters(
    Index 0 Param 1=&MakeFloatFromVector(Vector=$Particles.Scale, Channel=::NewEnumerator0)
    Index 0 Param 4=&FloatFromCurve(
     FloatCurve=DataInterface() {propertyValues {Curve {keys [
         {interpMode=RCIM_Cubic, leaveTangent=4.4102, leaveTangentWeight=0.1524, tangentMode=RCTM_Break, time=0, value=0}
         {arriveTangent=0.4089, interpMode=RCIM_Cubic, leaveTangent=0.4089, tangentMode=RCTM_User, time=0.572, value=1.375}
         {tangentMode=RCTM_User, time=1.3, value=1}
        ]}}}
     CurveIndex=&Add_Float(A=$Particles.Age, B=&Multiply_FloatByInt(Float=100, Integer=$User.Deployed))
     Scale Curve=&MakeCustomFloatFromBool(Bool=$User.IsFrontEnd, True Float=1.1, False Float=1)
    )
    Index 1 Param 1=&MakeFloatFromVector(Vector=$Particles.FeetPos, Channel=::NewEnumerator0)
    Index 1 Param 2=&MakeFloatFromVector(Vector=$Particles.FeetPos, Channel=::NewEnumerator1)
    Index 1 Param 3=&MakeFloatFromVector(Vector=$Particles.FeetPos, Channel=::NewEnumerator2)
    Index 1 Param 4=&MakeCustomFloatFromBool(
     Bool=$User.IsFrontEnd
     True Float=&FloatFromCurve(
      FloatCurve=DataInterface() {propertyValues {Curve {keys [<at 0 2.3>,<at 1 1>]}}}
      CurveIndex=&ClampFloat(Float=&Add_Float(A=$Particles.Age, B=&Multiply_FloatByInt(Float=100, Integer=$User.Deployed)))
     )
     False Float=&FloatFromCurve(
      FloatCurve=DataInterface() {propertyValues {Curve {keys [<at 0 3>,<at 1 1>]}}}
      CurveIndex=&ClampFloat(Float=&Add_Float(A=$Particles.Age, B=&Multiply_FloatByInt(Float=100, Integer=$User.Deployed)))
     )
    )
    Index 2 Param 1=&MakeFloatFromVector(Vector=$Particles.Feet2Pos, Channel=::NewEnumerator0)
    Index 2 Param 2=&MakeFloatFromVector(Vector=$Particles.Feet2Pos, Channel=::NewEnumerator1)
    Index 2 Param 3=&MakeFloatFromVector(Vector=$Particles.Feet2Pos, Channel=::NewEnumerator2)
    Index 2 Param 4=&MakeCustomFloatFromBool(
     Bool=$User.IsFrontEnd
     True Float=&FloatFromCurve(
      FloatCurve=DataInterface() {propertyValues {Curve {keys [
          {arriveTangent=-0.3111, interpMode=RCIM_Cubic, leaveTangent=-0.3111, tangentMode=RCTM_User, time=0, value=1}
          {arriveTangent=-0.4837, interpMode=RCIM_Cubic, leaveTangent=-0.4837, tangentMode=RCTM_User, time=0.9, value=0.65}
          {arriveTangent=-0.2194, interpMode=RCIM_Cubic, leaveTangent=-0.2194, tangentMode=RCTM_User, time=1.0858, value=-0.3}
          {arriveTangent=0.0139, interpMode=RCIM_Cubic, leaveTangent=0.0139, tangentMode=RCTM_User, time=1.3, value=0.15}
          {arriveTangent=0.0474, interpMode=RCIM_Cubic, leaveTangent=0.0474, tangentMode=RCTM_User, time=1.475, value=-0.13}
          {interpMode=RCIM_Cubic, leaveTangent=0.1333, time=1.7, value=0.07}
          <at 2 -0.08 RCIM_Cubic>
         ]}}}
      CurveIndex=&Add_Float(A=$Particles.Age, B=&Multiply_FloatByInt(Float=100, Integer=$User.Deployed))
     )
     False Float=&FloatFromCurve(
      FloatCurve=DataInterface() {propertyValues {Curve {keys [
          {arriveTangent=-0.3111, interpMode=RCIM_Cubic, leaveTangent=-0.3111, tangentMode=RCTM_User, time=0, value=1}
          {arriveTangent=-0.4837, interpMode=RCIM_Cubic, leaveTangent=-0.4837, tangentMode=RCTM_User, time=0.9, value=0.85}
          {arriveTangent=-0.2194, interpMode=RCIM_Cubic, leaveTangent=-0.2194, tangentMode=RCTM_User, time=1.0858, value=-0.3}
          {arriveTangent=0.0139, interpMode=RCIM_Cubic, leaveTangent=0.0139, tangentMode=RCTM_User, time=1.3, value=0.15}
          {arriveTangent=0.0474, interpMode=RCIM_Cubic, leaveTangent=0.0474, tangentMode=RCTM_User, time=1.475, value=-0.13}
          {interpMode=RCIM_Cubic, leaveTangent=0.1333, time=1.7, value=0.07}
          <at 2 -0.08 RCIM_Cubic>
         ]}}}
      CurveIndex=&Add_Float(A=$Particles.Age, B=&Multiply_FloatByInt(Float=100, Integer=$User.Deployed))
     )
    )
    Index 3 Param 1=&MakeCustomFloatFromBool(
     Bool=$User.IsFrontEnd
     True Float=&FloatFromCurve(
      FloatCurve=DataInterface() {propertyValues {Curve {keys [<at 0 3>,<at 0.9 2>,<at 1.5 0>]}}}
      CurveIndex=&Add_Float(A=$Particles.Age, B=&Multiply_FloatByInt(Float=100, Integer=$User.Deployed))
      Scale Curve=0.04
     )
     False Float=&FloatFromCurve(
      FloatCurve=DataInterface() {propertyValues {Curve {keys [<at 0 1>,<at 1.5 0>]}}}
      CurveIndex=&Add_Float(A=$Particles.Age, B=&Multiply_FloatByInt(Float=100, Integer=$User.Deployed))
      Scale Curve=0.05
     )
    )
    Index 3 Param 2=&MakeCustomFloatFromBool(Bool=$User.IsFrontEnd, True Float=1, False Float=0)
    Index 3 Param 4=$Particles.FN_Timer.Value
    Module.Param0WriteEnabled=true
    Module.Param0WriteEnabled1=true
    Module.Param0WriteEnabled2=true
    Module.Param0WriteEnabled3=true
    Module.Param1WriteEnabled=true
    Module.Param1WriteEnabled1=true
    Module.Param1WriteEnabled2=true
    Module.Param1WriteEnabled3=true
    Module.Param2WriteEnabled=true
    Module.Param2WriteEnabled1=true
    Module.Param2WriteEnabled2=true
    Module.Param2WriteEnabled3=true
    Module.Param3WriteEnabled=true
    Module.Param3WriteEnabled1=true
    Module.Param3WriteEnabled2=true
    Module.Param3WriteEnabled3=true
    Write Parameter Index 1=true
    Write Parameter Index 2=true
    Write Parameter Index 3=true
   )
   /SetVariables(Particles.MeshOrientation2:Quat4f=&MultiplyQuaternion(
     Quaternion A=&MakeQuatFromAxisAngle(
      Axis=<1 0 0>
      AngleInDegrees=&FloatFromCurve(
       FloatCurve=DataInterface() {propertyValues {Curve {keys [
           {arriveTangent=0.7701, interpMode=RCIM_Cubic, leaveTangent=0.7701, tangentMode=RCTM_User, time=0, value=-0.5}
           {arriveTangent=0.1957, interpMode=RCIM_Cubic, leaveTangent=0.1957, tangentMode=RCTM_User, time=1.15, value=0}
          ]}}}
       CurveIndex=&ClampFloat(Float=&Subtract_Float(
         A=&Add_Float(A=$Particles.Age, B=&Multiply_FloatByInt(Float=100, Integer=$User.Deployed))
         B=0.5
        ))
       Scale Curve=50
      )
     )
     Quaternion B=$Particles.MeshOrientation
    )) {isSetParameters=true, module=SetVariables_#4}
   /SetVariables(Particles.Position=&Lerp_Vector(
     A=<-3 10 -7>
     B=Linked() {linkedVariable {name=Particles.Initial.Position, type=NiagaraPosition}}
     Alpha=&ClampFloat(Float=&Divide_Float(
       A=&Add_Float(A=$Particles.Age, B=&Multiply_FloatByInt(Float=100, Integer=$User.Deployed))
       B=3
      ))
    )) {isSetParameters=true, module=SetVariables_#5}
   &ScaleColor(
    Scale RGB=&VectorFromFloat(Value=&FNTimeOfDayScalar(Night Scale=1, Morning Scale=1, Evening Scale=1))
    ScaleA=true
    ScaleRGB=true
   ) {module=ScaleColor003}
   &ScaleColor(
    Scale RGB=&VectorFromFloat(Value=&MakeCustomFloatFromBool(Bool=$User.IsFrontEnd, True Float=1, False Float=1))
    ScaleA=true
    ScaleRGB=true
   )
  ]
  rendererClasses [&NiagaraMeshRendererProperties]
  renderers [
   {rendererClass=&NiagaraMeshRendererProperties, rendererIndex=0}
   {rendererClass=&NiagaraMeshRendererProperties, rendererIndex=1}
  ]
 }
 Cloud001 {
  emitterUpdateScript [
   &EmitterState(
    Inactive Response=::NewEnumerator1
    Life Cycle Mode=::NewEnumerator1
    Loop Behavior=::NewEnumerator1
    Loop Delay=&MakeCustomFloatFromBool(
     Bool=&SetBoolByIntComparison(A=$User.Deployed, B=0, Comparison Type=::NewEnumerator0)
     True Float=0
     False Float=1.1
    )
    Loop Duration Mode=::NewEnumerator1
    UseLoopDelay=true
   )
   &SpawnBurst_Instantaneous(Use Spawn Probability=false, UseLoopCountLimit=false)
  ]
  particleSpawnScript [
   &InitializeParticle(
    Color Mode=::NewEnumerator0
    Lifetime=5
    Mesh Scale=&Multiply_VectorByFloat(Vector=<0.1 0.1 0.1>, Float=1.3)
    Mesh Scale Mode=::NewEnumerator4
    Position Offset=<0 -15 -8>
    Position Offset Coordinate Space=::Local
    UsePositionOffset=true
    Write Material Random=true
   )
   &InitialMeshOrientation(Mesh Axis To Orient=<0 1 0>, Mesh Orientation Mode=::NewEnumerator2, Orientation Coordinate Space=::Local, Use Rotation Vector=false)
  ]
  particleUpdateScript [
   &ParticleState(Kill Particles When Lifetime Has Elapsed=false, Let Infinitely Lived Particles Die When Emitter Deactivates=true)
   &ScaleMeshSize(Scale Factor=&VectorFromFloat(Value=&MakeCustomFloatFromBool(Bool=$User.IsFrontEnd, True Float=0.9, False Float=1)))
   /SetVariables(Particles.FeetPos:Vector3f=&TransformPosition(
     Position=&Add_Vector(A=Linked() {linkedVariable {name=Particles.Position, type=NiagaraPosition}}, B=<2.5 -15 20>)
     Source Space=::Local
     Destination Space=::World
    )) {isSetParameters=true, module=SetVariables_#2}
   /SetVariables(Particles.Feet2Pos:Vector3f=&TransformPosition(
     Position=&Add_Vector(A=Linked() {linkedVariable {name=Particles.Position, type=NiagaraPosition}}, B=<2.5 45 10>)
     Source Space=::Local
     Destination Space=::World
    )) {isSetParameters=true, module=SetVariables_#3}
   &FN_Timer(Increment Multiplier:NiagaraFloat=&FloatFromCurve(
     FloatCurve=DataInterface() {propertyValues {Curve {keys [{interpMode=RCIM_Cubic, tangentMode=RCTM_User, time=0, value=5.5}, {interpMode=RCIM_Cubic, tangentMode=RCTM_User, time=2, value=1}]}}}
     CurveIndex=&Add_Float(A=$Particles.Age, B=&Multiply_FloatByInt(Float=100, Integer=$User.Deployed))
     Scale Curve=&Lerp_Float(
      A=1
      B=&FloatFromCurve(
       FloatCurve=DataInterface() {propertyValues {Curve {keys [<at 0 1>,<at 2 2.35>]}}}
       CurveIndex=&Add_Float(A=$Particles.Age, B=&Multiply_FloatByInt(Float=100, Integer=$User.Deployed))
      )
      Alpha=&ClampFloat(Float=$System.InterpolateOverTime.InterpolateOverTime_Float.Moving Average)
     )
    ))
   &DynamicMaterialParameters(
    Index 0 Param 1=&MakeFloatFromVector(Vector=$Particles.Scale, Channel=::NewEnumerator0)
    Index 0 Param 4=&FloatFromCurve(
     FloatCurve=DataInterface() {propertyValues {Curve {keys [
         {interpMode=RCIM_Cubic, leaveTangent=4.4102, leaveTangentWeight=0.1524, tangentMode=RCTM_Break, time=0, value=0}
         {arriveTangent=0.4089, interpMode=RCIM_Cubic, leaveTangent=0.4089, tangentMode=RCTM_User, time=0.572, value=1.375}
         {tangentMode=RCTM_User, time=1.3, value=1}
        ]}}}
     CurveIndex=&Add_Float(A=$Particles.Age, B=&Multiply_FloatByInt(Float=100, Integer=$User.Deployed))
     Scale Curve=&MakeCustomFloatFromBool(Bool=$User.IsFrontEnd, True Float=1.1, False Float=1)
    )
    Index 1 Param 1=&MakeFloatFromVector(Vector=$Particles.FeetPos, Channel=::NewEnumerator0)
    Index 1 Param 2=&MakeFloatFromVector(Vector=$Particles.FeetPos, Channel=::NewEnumerator1)
    Index 1 Param 3=&MakeFloatFromVector(Vector=$Particles.FeetPos, Channel=::NewEnumerator2)
    Index 1 Param 4=&MakeCustomFloatFromBool(
     Bool=$User.IsFrontEnd
     True Float=&FloatFromCurve(
      FloatCurve=DataInterface() {propertyValues {Curve {keys [<at 0 2.3>,<at 1 1>]}}}
      CurveIndex=&ClampFloat(Float=&Add_Float(A=$Particles.Age, B=&Multiply_FloatByInt(Float=100, Integer=$User.Deployed)))
     )
     False Float=&FloatFromCurve(
      FloatCurve=DataInterface() {propertyValues {Curve {keys [<at 0 3>,<at 1 1>]}}}
      CurveIndex=&ClampFloat(Float=&Add_Float(A=$Particles.Age, B=&Multiply_FloatByInt(Float=100, Integer=$User.Deployed)))
     )
    )
    Index 2 Param 1=&MakeFloatFromVector(Vector=$Particles.Feet2Pos, Channel=::NewEnumerator0)
    Index 2 Param 2=&MakeFloatFromVector(Vector=$Particles.Feet2Pos, Channel=::NewEnumerator1)
    Index 2 Param 3=&MakeFloatFromVector(Vector=$Particles.Feet2Pos, Channel=::NewEnumerator2)
    Index 2 Param 4=&MakeCustomFloatFromBool(
     Bool=$User.IsFrontEnd
     True Float=&FloatFromCurve(
      FloatCurve=DataInterface() {propertyValues {Curve {keys [
          {arriveTangent=-0.3111, interpMode=RCIM_Cubic, leaveTangent=-0.3111, tangentMode=RCTM_User, time=0, value=1}
          {arriveTangent=-0.4837, interpMode=RCIM_Cubic, leaveTangent=-0.4837, tangentMode=RCTM_User, time=0.9, value=0.65}
          {arriveTangent=-0.2194, interpMode=RCIM_Cubic, leaveTangent=-0.2194, tangentMode=RCTM_User, time=1.0858, value=-0.3}
          {arriveTangent=0.0139, interpMode=RCIM_Cubic, leaveTangent=0.0139, tangentMode=RCTM_User, time=1.3, value=0.15}
          {arriveTangent=0.0474, interpMode=RCIM_Cubic, leaveTangent=0.0474, tangentMode=RCTM_User, time=1.475, value=-0.13}
          {interpMode=RCIM_Cubic, leaveTangent=0.1333, time=1.7, value=0.07}
          <at 2 -0.08 RCIM_Cubic>
         ]}}}
      CurveIndex=&Add_Float(A=$Particles.Age, B=&Multiply_FloatByInt(Float=100, Integer=$User.Deployed))
     )
     False Float=&FloatFromCurve(
      FloatCurve=DataInterface() {propertyValues {Curve {keys [
          {arriveTangent=-0.3111, interpMode=RCIM_Cubic, leaveTangent=-0.3111, tangentMode=RCTM_User, time=0, value=1}
          {arriveTangent=-0.4837, interpMode=RCIM_Cubic, leaveTangent=-0.4837, tangentMode=RCTM_User, time=0.9, value=0.85}
          {arriveTangent=-0.2194, interpMode=RCIM_Cubic, leaveTangent=-0.2194, tangentMode=RCTM_User, time=1.0858, value=-0.3}
          {arriveTangent=0.0139, interpMode=RCIM_Cubic, leaveTangent=0.0139, tangentMode=RCTM_User, time=1.3, value=0.15}
          {arriveTangent=0.0474, interpMode=RCIM_Cubic, leaveTangent=0.0474, tangentMode=RCTM_User, time=1.475, value=-0.13}
          {interpMode=RCIM_Cubic, leaveTangent=0.1333, time=1.7, value=0.07}
          <at 2 -0.08 RCIM_Cubic>
         ]}}}
      CurveIndex=&Add_Float(A=$Particles.Age, B=&Multiply_FloatByInt(Float=100, Integer=$User.Deployed))
     )
    )
    Index 3 Param 1=&MakeCustomFloatFromBool(
     Bool=$User.IsFrontEnd
     True Float=&FloatFromCurve(
      FloatCurve=DataInterface() {propertyValues {Curve {keys [<at 0 3>,<at 0.9 2>,<at 1.5 0>]}}}
      CurveIndex=&Add_Float(A=$Particles.Age, B=&Multiply_FloatByInt(Float=100, Integer=$User.Deployed))
      Scale Curve=0.04
     )
     False Float=&FloatFromCurve(
      FloatCurve=DataInterface() {propertyValues {Curve {keys [<at 0 1>,<at 1.5 0>]}}}
      CurveIndex=&Add_Float(A=$Particles.Age, B=&Multiply_FloatByInt(Float=100, Integer=$User.Deployed))
      Scale Curve=0.05
     )
    )
    Index 3 Param 2=&MakeCustomFloatFromBool(Bool=$User.IsFrontEnd, True Float=1, False Float=0)
    Index 3 Param 4=$Particles.FN_Timer.Value
    Module.Param0WriteEnabled=true
    Module.Param0WriteEnabled1=true
    Module.Param0WriteEnabled2=true
    Module.Param0WriteEnabled3=true
    Module.Param1WriteEnabled=true
    Module.Param1WriteEnabled1=true
    Module.Param1WriteEnabled2=true
    Module.Param1WriteEnabled3=true
    Module.Param2WriteEnabled=true
    Module.Param2WriteEnabled1=true
    Module.Param2WriteEnabled2=true
    Module.Param2WriteEnabled3=true
    Module.Param3WriteEnabled=true
    Module.Param3WriteEnabled1=true
    Module.Param3WriteEnabled2=true
    Module.Param3WriteEnabled3=true
    Write Parameter Index 1=true
    Write Parameter Index 2=true
    Write Parameter Index 3=true
   )
   /SetVariables(Particles.MeshOrientation2:Quat4f=&MultiplyQuaternion(
     Quaternion A=&MakeQuatFromAxisAngle(
      Axis=<1 0 0>
      AngleInDegrees=&FloatFromCurve(
       FloatCurve=DataInterface() {propertyValues {Curve {keys [
           {arriveTangent=0.7701, interpMode=RCIM_Cubic, leaveTangent=0.7701, tangentMode=RCTM_User, time=0, value=-0.5}
           {arriveTangent=0.1957, interpMode=RCIM_Cubic, leaveTangent=0.1957, tangentMode=RCTM_User, time=1.15, value=0}
          ]}}}
       CurveIndex=&ClampFloat(Float=&Subtract_Float(
         A=&Add_Float(A=$Particles.Age, B=&Multiply_FloatByInt(Float=100, Integer=$User.Deployed))
         B=0.5
        ))
       Scale Curve=50
      )
     )
     Quaternion B=$Particles.MeshOrientation
    )) {isSetParameters=true, module=SetVariables_#4}
   /SetVariables(Particles.Position=&Lerp_Vector(
     A=<-3 10 -7>
     B=Linked() {linkedVariable {name=Particles.Initial.Position, type=NiagaraPosition}}
     Alpha=&ClampFloat(Float=&Divide_Float(
       A=&Add_Float(A=$Particles.Age, B=&Multiply_FloatByInt(Float=100, Integer=$User.Deployed))
       B=3
      ))
    )) {isSetParameters=true, module=SetVariables_#5}
   &ScaleColor(
    Scale RGB=&VectorFromFloat(Value=&FNTimeOfDayScalar(Night Scale=1, Morning Scale=1, Evening Scale=1))
    ScaleA=true
    ScaleRGB=true
   ) {module=ScaleColor003}
   &ScaleColor(
    Scale RGB=&VectorFromFloat(Value=&MakeCustomFloatFromBool(Bool=$User.IsFrontEnd, True Float=2, False Float=1))
    ScaleA=true
    ScaleRGB=true
   )
  ]
  rendererClasses [&NiagaraMeshRendererProperties]
  renderers [
   {rendererClass=&NiagaraMeshRendererProperties, rendererIndex=0}
   {rendererClass=&NiagaraMeshRendererProperties, rendererIndex=1}
  ]
 }
}
summary {
 systemName=FX_Thrive
 emitters [
  {emitterName=Cloud, rendererClasses [&NiagaraMeshRendererProperties]}
  {emitterName=Cloud001, rendererClasses [&NiagaraMeshRendererProperties]}
 ]
 userVariables [
  User.Deployed {type=NiagaraInt32}
  User.ForwardDot {type=NiagaraFloat, defaultValue=0, description=""}
  User.IsFrontEnd {type=NiagaraBool}
  User.IsWhiteVersion {type=NiagaraBool, defaultValue=true}
 ]
}
