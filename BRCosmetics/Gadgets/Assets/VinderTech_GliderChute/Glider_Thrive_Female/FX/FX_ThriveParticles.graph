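# Niagara system for the Thrive (female) VinderTech glider: mesh layers that grow in on deployment, with drifting mesh particles and a brief spinning mesh.
# Long-lived burst meshes follow the owner, feeding a lagged, angle-clamped target point, timers and age curves into dynamic material parameters.
# The drifting particles read another emitter's attributes to steer their forces.
# User flags (Deployed, GrowDelay, IsFrontEnd, IsWhiteVersion) and time-of-day scalars adjust timing, scale and brightness.
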NiagaraSystem /BRCosmetics/Gadgets/Assets/VinderTech_GliderChute/Glider_Thrive_Female/FX/FX_ThriveParticles
dependencies {
 usedDataInterfaces [
  &NiagaraDataInterfaceParticleRead
  &NiagaraDataInterfaceCurve
  &NiagaraDataInterfaceColorCurve
  &NiagaraDataInterfaceCamera
 ]
 usedDynamicInputs [&Boolean_NotOperation]
 usedModules [
  &self:SystemSpawnScript.SystemScriptSource.SystemScriptGraph.NiagaraNodeAssignment_0.SetVariables_#1
  &SystemState
 ]
}
emitters {
 Empty {
  emitterUpdateScript [
   &EmitterState(
    Inactive Response=::NewEnumerator1
    Life Cycle Mode=::NewEnumerator1
    Loop Behavior=::NewEnumerator1
    Loop Delay=&MakeCustomFloatFromBool(
     Bool=&SetBoolByIntComparison(A=$User.Deployed, B=0, Comparison Type=::NewEnumerator0)
     True Float=0
     False Float=2.4
    )
    Loop Duration=10000
    UseLoopDelay=true
   )
   &SpawnRate(
    SpawnRate=&Lerp_Float(
     A=&MakeCustomFloatFromBool(Bool=$User.IsFrontEnd, True Float=3.5, False Float=0)
     B=3.5
     Alpha=&ClampFloat(Float=$System.InterpolateOverTime.InterpolateOverTime_Float.Moving Average)
    )
    Use Spawn Probability=false
   )
  ]
  particleSpawnScript [
   &InitializeParticle(
    Color Mode=::NewEnumerator0
    Lifetime Max=0.75
    Lifetime Min=0.5
    Lifetime Mode=::NewEnumerator1
    Mesh Scale Mode=::NewEnumerator3
    Mesh Uniform Scale=&Lerp_Float(
     A=0.2
     B=0.3
     Alpha=&FloatFromCurve(
      FloatCurve=DataInterface() {propertyValues {Curve {keys [
          {interpMode=RCIM_Cubic, tangentMode=RCTM_User, time=0, value=0}
          {arriveTangent=3.0994, interpMode=RCIM_Cubic, leaveTangent=3.0994, tangentMode=RCTM_User, time=1, value=1}
         ]}}}
      CurveIndex=&RandomRangeFloat(Minimum=0, Maximum=1)
     )
    )
    UsePositionOffset=false
    Write Material Random=true
   )
   &SphereLocation(
    Hemisphere Y=true
    Hemisphere Z=true
    MaskSpawnGroup=false
    Offset=<0 55.3527 0>
    Override Seed=false
    Sphere Radius=50
    Surface Only=false
    UseNonUniformScale=true
    UseOffset=true
    UseSphereOrientation=false
   ) {enabled=false}
   &BoxLocation(Box Offset=<0 50 0>, Box Size=<100 1 1>, MaskSpawnGroup=false, Module.Override Seed=false, Rotate Box=false, Surface Only=false)
   &AddVelocity(Velocity=&MakeCustomVectorFromBool(Bool=$User.IsFrontEnd, True Vector=<0 150 0>, False Vector=<0 150 0>))
   /SetVariables(Particles.MeshIndex=&RandomRangeInt(
     Minimum=0
     Maximum=&RandomRangeInt(Minimum=0, Maximum=&RandomRangeInt(Minimum=1, Maximum=1))
    )) {isSetParameters=true}
   &InitialMeshOrientation(Use Rotation Vector=false)
  ]
  particleUpdateScript [
   &PointAttractionForce(
    Attraction Radius=5000
    Attraction Strength=20
    Attractor Position Offset:Vector3f=<0 200 50>
    Falloff Exponent=0.4
    Kill Within Radius=false
    Module.Use Falloff=true
    UseAttractorOffset=true
   ) {enabled=false}
   &self:ReadOtherParticleAttri(New Particle Attribute Reader=DataInterface() {propertyValues {EmitterBinding {bindingMode=Other, emitterName=Empty003}}}) {module=ReadOtherParticleAttri}
   /Niagara/Modules/Update/Forces/LineAttractionForce(
    Attraction Strength=150
    Line End:Vector3f=<0 500 0>
    Line Segment in Localspace=true
    Line Start:Vector3f=<0 0 0>
    Use Attraction Falloff=false
   )
   &LinearForce(Force=&Lerp_Vector(
     A=&Lerp_Vector(
      A=<850 -200 0>
      B=<-850 -200 0>
      Alpha=&ClampFloat(Float=&DotProduct(
        Vector A=&NormalizeVector(
         Vector To Normalize=&Subtract_Vector(A=$Particles.Value1, B=<0 0 0>)
         Safe Normalize=true
         Safe Fallback Threshold=0
         Safe Fallback Vector=<0 0 0.001>
        )
        Vector B=<-1 0 0>
       ))
     )
     B=<0 500 0>
     Alpha=&DotProduct(
      Vector A=&NormalizeVector(
       Vector To Normalize=&Subtract_Vector(A=$Particles.Value1, B=<0 0 0>)
       Safe Normalize=true
       Safe Fallback Threshold=0
       Safe Fallback Vector=<0 0 0.001>
      )
      Vector B=<0 1 0>
     )
    ))
   &SolveForcesAndVelocity(Clamp Velocity=false, Limit Acceleration=false)
   &ParticleState()
   &UpdateMeshOrientation(
    Rotation Rate=&RandomRangeFloat(Minimum=0.3, Maximum=0.5)
    Rotation Vector=&RandomRangeVector(Minimum=<0 0 0>, Maximum=<1 1 1>)
   )
   &ScaleMeshSize(Scale Factor=&VectorFromFloat(Value=&FloatFromCurve(
      FloatCurve=DataInterface() {propertyValues {Curve {keys [
          {arriveTangent=11.2223, interpMode=RCIM_Cubic, leaveTangent=11.2223, tangentMode=RCTM_User, time=-0.0026, value=0.2}
          {arriveTangent=-3.6459, interpMode=RCIM_Cubic, leaveTangent=-3.6459, tangentMode=RCTM_User, time=0.3983, value=1.5254}
          {tangentMode=RCTM_User, time=0.9974, value=0.901}
         ]}}}
      Scale Curve=&MakeCustomFloatFromBool(
       Bool=&SetBoolByIntComparison(A=$Particles.MeshIndex, B=0, Comparison Type=::NewEnumerator4)
       True Float=0.8
       False Float=1.13
      )
     )))
   &ScaleColor(
    Scale Alpha=&FloatFromCurve(FloatCurve=DataInterface() {propertyValues {Curve {keys [
         {interpMode=RCIM_Cubic, leaveTangent=5.7537, tangentMode=RCTM_Break, time=0, value=0}
         <at 0.4 1>
         {arriveTangent=-0.0405, interpMode=RCIM_Cubic, leaveTangent=-0.0405, tangentMode=RCTM_User, time=0.5354, value=1}
         {arriveTangent=-2.2563, interpMode=RCIM_Cubic, leaveTangent=-2.2563, tangentMode=RCTM_User, time=1, value=0}
        ]}}})
    ScaleA=true
    ScaleRGB=true
   )
   &ScaleColor(
    Scale RGB=&VectorFromFloat(Value=&FNTimeOfDayScalar(Night Scale=1, Morning Scale=1, Evening Scale=1))
    ScaleA=true
    ScaleRGB=true
   ) {module=ScaleColor003}
   &ScaleMeshSize(Scale Factor=&VectorFromFloat(Value=&MakeCustomFloatFromBool(Bool=$User.IsFrontEnd, True Float=0.8, False Float=1))) {module=ScaleMeshSize001}
  ]
  rendererClasses [&NiagaraMeshRendererProperties]
  renderers [
   {rendererClass=&NiagaraMeshRendererProperties, rendererIndex=0}
   {rendererClass=&NiagaraMeshRendererProperties, rendererIndex=1}
  ]
 }
 Empty001 {
  emitterUpdateScript [
   &EmitterState(
    Inactive Response=::NewEnumerator1
    Life Cycle Mode=::NewEnumerator1
    Loop Behavior=::NewEnumerator1
    Loop Delay=&MakeCustomFloatFromBool(
     Bool=&SetBoolByIntComparison(A=$User.Deployed, B=0, Comparison Type=::NewEnumerator0)
     True Float=0
     False Float=$User.GrowDelay
    )
    Loop Duration Mode=::NewEnumerator1
    UseLoopDelay=true
   )
   &SpawnBurst_Instantaneous(Use Spawn Probability=false, UseLoopCountLimit=false)
  ]
  particleSpawnScript [&InitializeParticle(Color Mode=::NewEnumerator0, Lifetime=9999999, Mesh Scale Mode=::NewEnumerator3, UsePositionOffset=true, Write Material Random=true)]
  particleUpdateScript [
   &ParticleState()
   /SetVariables(Particles.InterpolatePos:Vector3f=/Niagara/DynamicInputs/Transforms/TransformPosition(Position=<0 300 0>, Source Space=::Local, Destination Space=::Simulation)) {isSetParameters=true, module=SetVariables_#2}
   &InterpolateOverTime(Rate Of Change=2, Target Vector Value:Vector3f=$Particles.InterpolatePos, Variable Type=::NewEnumerator2)
   &self:Clamp(
    In:Vector3f=&Lerp_Vector(
     A=$Particles.InterpolatePos
     B=$Particles.InterpolateOverTime.InterpolateOverTime_V3_.Moving Average
     Alpha=&ClampFloat(Float=&Divide_Float(A=$Particles.Age, B=3))
    )
    Max Angle=75
   ) {module=Clamp}
   /SetVariables(Particles.InterpolateOverTime.InterpolateOverTime_V3_.Moving Average:Vector3f=$Particles.Clamp.Clamped) {isSetParameters=true, module=SetVariables_#3}
   /SetVariables(Particles.Position2:Vector3f=&Add_Vector(
     A=Linked() {linkedVariable {name=Engine.Owner.Position, type=NiagaraPosition}}
     B=&Multiply_VectorByFloat(
      Vector=&NormalizeVector(
       Vector To Normalize=&Subtract_Vector(A=$Particles.Clamp.Clamped, B=Linked() {linkedVariable {name=Engine.Owner.Position, type=NiagaraPosition}})
       Safe Normalize=true
       Safe Fallback Threshold=0
       Safe Fallback Vector=<0 0 0.001>
      )
      Float=300
     )
    )) {isSetParameters=true, module=SetVariables_#4}
   /SetVariables(Particles.MeshOrientation=$Engine.Owner.Rotation) {isSetParameters=true, module=SetVariables_#5}
   /SetVariables(Particles.OffsetAge:NiagaraFloat=&Subtract_Float(
     A=&Add_Float(A=$Particles.Age, B=&Multiply_FloatByInt(Float=100, Integer=$User.Deployed))
     B=1.5
    )) {isSetParameters=true}
   &FN_Timer(Increment Multiplier:NiagaraFloat=&FloatFromCurve(
     FloatCurve=DataInterface() {propertyValues {Curve {keys [{interpMode=RCIM_Cubic, tangentMode=RCTM_User, time=0, value=1.5}, {interpMode=RCIM_Cubic, tangentMode=RCTM_User, time=2, value=1}]}}}
     CurveIndex=$Particles.OffsetAge
     Scale Curve=&Lerp_Float(A=0.9, B=1.3, Alpha=&ClampFloat(Float=$System.InterpolateOverTime.InterpolateOverTime_Float.Moving Average))
    ))
   &DynamicMaterialParameters(
    Index 0 Param 1=&MakeFloatFromVector(Vector=$Particles.Position2, Channel=::NewEnumerator0)
    Index 0 Param 2=&MakeFloatFromVector(Vector=$Particles.Position2, Channel=::NewEnumerator1)
    Index 0 Param 3=&MakeFloatFromVector(Vector=$Particles.Position2, Channel=::NewEnumerator2)
    Index 0 Param 4=&MakeFloatFromVector(Vector=&TransformVector(Vector=<0 1 0>, Destination Space=::World), Channel=::NewEnumerator0)
    Index 1 Param 1=&MakeFloatFromVector(Vector=&TransformVector(Vector=<0 1 0>, Destination Space=::World), Channel=::NewEnumerator1)
    Index 1 Param 2=&MakeFloatFromVector(Vector=&TransformVector(Vector=<0 1 0>, Destination Space=::World), Channel=::NewEnumerator2)
    Index 1 Param 3=&MakeFloatFromVector(Vector=Linked() {linkedVariable {name=Engine.Owner.Position, type=NiagaraPosition}}, Channel=::NewEnumerator0)
    Index 1 Param 4=&MakeFloatFromVector(Vector=Linked() {linkedVariable {name=Engine.Owner.Position, type=NiagaraPosition}}, Channel=::NewEnumerator1)
    Index 2 Param 1=&MakeFloatFromVector(Vector=Linked() {linkedVariable {name=Engine.Owner.Position, type=NiagaraPosition}}, Channel=::NewEnumerator2)
    Index 2 Param 2=1
    Index 2 Param 3=&FloatFromCurve(
     FloatCurve=DataInterface() {propertyValues {Curve {keys [
         {arriveTangent=-0.4447, interpMode=RCIM_Cubic, leaveTangent=-0.4447, tangentMode=RCTM_User, time=0, value=1}
         {arriveTangent=-1.8533, interpMode=RCIM_Cubic, leaveTangent=-1.8533, tangentMode=RCTM_User, time=1, value=0}
        ]}}}
     CurveIndex=$Particles.OffsetAge
    )
    Index 2 Param 4=&FloatFromCurve(
     FloatCurve=DataInterface() {propertyValues {Curve {keys [
         {arriveTangent=-0.1185, interpMode=RCIM_Cubic, leaveTangent=-0.1185, tangentMode=RCTM_User, time=0, value=0.01}
         {arriveTangent=1.9079, interpMode=RCIM_Cubic, leaveTangent=1.0524, tangentMode=RCTM_Break, time=1, value=1}
        ]}}}
     CurveIndex=$Particles.OffsetAge
    )
    Index 3 Param 1=&Add_Float(A=$Particles.FN_Timer.Value, B=1.1403)
    Module.Param0WriteEnabled=true
    Module.Param0WriteEnabled1=true
    Module.Param0WriteEnabled2=true
    Module.Param0WriteEnabled3=true
    Module.Param1WriteEnabled=true
    Module.Param1WriteEnabled1=true
    Module.Param1WriteEnabled2=true
    Module.Param1WriteEnabled3=true
    Module.Param2WriteEnabled=true
    Module.Param2WriteEnabled1=true
    Module.Param2WriteEnabled2=true
    Module.Param2WriteEnabled3=true
    Module.Param3WriteEnabled=true
    Module.Param3WriteEnabled1=true
    Module.Param3WriteEnabled2=true
    Module.Param3WriteEnabled3=true
    Write Parameter Index 1=true
    Write Parameter Index 2=true
    Write Parameter Index 3=true
   )
   &ScaleMeshSize(Scale Factor=&MakeVector(
     X=&FloatFromCurve(
      FloatCurve=DataInterface() {propertyValues {Curve {keys [
          {arriveTangent=0.6276, interpMode=RCIM_Cubic, leaveTangent=0.6276, tangentMode=RCTM_User, time=0, value=0.75}
          {arriveTangent=0.2159, interpMode=RCIM_Cubic, leaveTangent=0.2159, tangentMode=RCTM_User, time=1, value=1}
         ]}}}
      CurveIndex=$Particles.OffsetAge
     )
     Y=&FloatFromCurve(
      FloatCurve=DataInterface() {propertyValues {Curve {keys [{tangentMode=RCTM_User, time=0.2, value=1}, {tangentMode=RCTM_User, time=1, value=1}]}}}
      CurveIndex=$Particles.OffsetAge
     )
     Z=1
    )) {module=ScaleMeshSize001}
   /SetVariables(Particles.Position=&Add_Vector(
     A=Linked() {linkedVariable {name=Engine.Owner.Position, type=NiagaraPosition}}
     B=&TransformVector(
      Vector=&MakeVector(
       X=0
       Y=&FloatFromCurve(
        FloatCurve=DataInterface() {propertyValues {Curve {keys [
            {arriveTangent=-0.2987, interpMode=RCIM_Cubic, leaveTangent=-0.2987, tangentMode=RCTM_User, time=0, value=0.75}
            {arriveTangent=-0.0552, interpMode=RCIM_Cubic, leaveTangent=-0.0552, tangentMode=RCTM_User, time=1, value=0}
           ]}}}
        CurveIndex=$Particles.OffsetAge
        Scale Curve=-60
       )
       Z=0
      )
      Destination Space=::World
     )
    )) {isSetParameters=true, module=SetVariables_#6}
   &ScaleMeshSize(Scale Factor=&MakeVector(
     X=&FloatFromCurve(FloatCurve=DataInterface() {propertyValues {Curve {keys [<at 0 0.3>,<at 1 1>]}}}, CurveIndex=$Particles.OffsetAge)
     Y=1
     Z=1
    )) {enabled=false}
   &ScaleColor(
    Scale Alpha=&FloatFromCurve(
     FloatCurve=DataInterface() {propertyValues {Curve {keys [
         {arriveTangent=13.7841, interpMode=RCIM_Cubic, leaveTangent=13.7841, tangentMode=RCTM_User, time=0, value=0}
         {interpMode=RCIM_Cubic, tangentMode=RCTM_User, time=0.1, value=1}
        ]}}}
     CurveIndex=$Particles.OffsetAge
    )
    ScaleA=true
    ScaleRGB=true
   )
   &ScaleColor(
    Scale RGB=&VectorFromFloat(Value=&FNTimeOfDayScalar(Night Scale=0.7, Morning Scale=0.4, Evening Scale=0.4))
    ScaleA=true
    ScaleRGB=true
   ) {module=ScaleColor001}
   &ScaleColor(
    Scale Alpha=&MakeCustomFloatFromBool(Bool=$User.IsFrontEnd, True Float=0.7, False Float=1)
    Scale RGB=&VectorFromFloat(Value=&MakeCustomFloatFromBool(
      Bool=$User.IsFrontEnd
      True Float=&MakeCustomFloatFromBool(Bool=$User.IsWhiteVersion, True Float=1.75, False Float=2)
      False Float=1
     ))
    ScaleA=true
    ScaleRGB=true
   ) {module=ScaleColor002}
  ]
  rendererClasses [&NiagaraMeshRendererProperties]
  renderers [
   {rendererClass=&NiagaraMeshRendererProperties, rendererIndex=0}
   {rendererClass=&NiagaraMeshRendererProperties, rendererIndex=1}
   {rendererClass=&NiagaraMeshRendererProperties, rendererIndex=2}
   {rendererClass=&NiagaraMeshRendererProperties, rendererIndex=3}
  ]
 }
 Empty002 {
  emitterUpdateScript [
   &EmitterState(
    Inactive Response=::NewEnumerator1
    Life Cycle Mode=::NewEnumerator1
    Loop Behavior=::NewEnumerator1
    Loop Delay=&MakeCustomFloatFromBool(
     Bool=&SetBoolByIntComparison(A=$User.Deployed, B=0, Comparison Type=::NewEnumerator0)
     True Float=0
     False Float=$User.GrowDelay
    )
    Loop Duration Mode=::NewEnumerator1
    UseLoopDelay=true
   )
   &SpawnBurst_Instantaneous(Use Spawn Probability=false, UseLoopCountLimit=false)
  ]
  particleSpawnScript [&InitializeParticle(Color Mode=::NewEnumerator0, Lifetime=9999999, Mesh Scale Mode=::NewEnumerator3, UsePositionOffset=true, Write Material Random=true)]
  particleUpdateScript [
   &ParticleState()
   /SetVariables(Particles.InterpolatePos:Vector3f=&TransformPosition(Position=<0 300 0>, Source Space=::Local, Destination Space=::Simulation)) {isSetParameters=true}
   &InterpolateOverTime(Rate Of Change=2, Target Vector Value:Vector3f=$Particles.InterpolatePos, Variable Type=::NewEnumerator2)
   &self:Clamp(
    In:Vector3f=&Lerp_Vector(
     A=$Particles.InterpolatePos
     B=$Particles.InterpolateOverTime.InterpolateOverTime_V3_.Moving Average
     Alpha=&ClampFloat(Float=&Divide_Float(A=$Particles.Age, B=3))
    )
    Max Angle=75
   ) {module=Clamp}
   /SetVariables(Particles.InterpolateOverTime.InterpolateOverTime_V3_.Moving Average:Vector3f=$Particles.Clamp.Clamped) {isSetParameters=true, module=SetVariables_#3}
   /SetVariables(Particles.Position2:Vector3f=&Add_Vector(
     A=Linked() {linkedVariable {name=Engine.Owner.Position, type=NiagaraPosition}}
     B=&Multiply_VectorByFloat(
      Vector=&NormalizeVector(
       Vector To Normalize=&Subtract_Vector(A=$Particles.Clamp.Clamped, B=Linked() {linkedVariable {name=Engine.Owner.Position, type=NiagaraPosition}})
       Safe Normalize=true
       Safe Fallback Threshold=0
       Safe Fallback Vector=<0 0 0.001>
      )
      Float=300
     )
    )) {isSetParameters=true, module=SetVariables_#4}
   /SetVariables(Particles.MeshOrientation=$Engine.Owner.Rotation) {isSetParameters=true, module=SetVariables_#5}
   /SetVariables(Particles.OffsetAge:NiagaraFloat=&Subtract_Float(
     A=&Add_Float(A=$Particles.Age, B=&Multiply_FloatByInt(Float=100, Integer=$User.Deployed))
     B=0.12
    )) {isSetParameters=true, module=SetVariables_#7}
   &FN_Timer(Increment Multiplier:NiagaraFloat=&Lerp_Float(A=0.9, B=1.3, Alpha=&ClampFloat(Float=$System.InterpolateOverTime.InterpolateOverTime_Float.Moving Average)))
   &DynamicMaterialParameters(
    Index 0 Param 1=&MakeFloatFromVector(Vector=$Particles.Position2, Channel=::NewEnumerator0)
    Index 0 Param 2=&MakeFloatFromVector(Vector=$Particles.Position2, Channel=::NewEnumerator1)
    Index 0 Param 3=&MakeFloatFromVector(Vector=$Particles.Position2, Channel=::NewEnumerator2)
    Index 0 Param 4=&MakeFloatFromVector(Vector=&TransformVector(Vector=<0 1 0>, Destination Space=::World), Channel=::NewEnumerator0)
    Index 1 Param 1=&MakeFloatFromVector(Vector=&TransformVector(Vector=<0 1 0>, Destination Space=::World), Channel=::NewEnumerator1)
    Index 1 Param 2=&MakeFloatFromVector(Vector=&TransformVector(Vector=<0 1 0>, Destination Space=::World), Channel=::NewEnumerator2)
    Index 1 Param 3=&MakeFloatFromVector(Vector=Linked() {linkedVariable {name=Engine.Owner.Position, type=NiagaraPosition}}, Channel=::NewEnumerator0)
    Index 1 Param 4=&MakeFloatFromVector(Vector=Linked() {linkedVariable {name=Engine.Owner.Position, type=NiagaraPosition}}, Channel=::NewEnumerator1)
    Index 2 Param 1=&MakeFloatFromVector(Vector=Linked() {linkedVariable {name=Engine.Owner.Position, type=NiagaraPosition}}, Channel=::NewEnumerator2)
    Index 2 Param 2=1
    Index 2 Param 4=&FloatFromCurve(
     FloatCurve=DataInterface() {propertyValues {Curve {keys [
         {arriveTangent=-0.1185, interpMode=RCIM_Cubic, leaveTangent=-0.1185, tangentMode=RCTM_User, time=0, value=0}
         {arriveTangent=1.9079, interpMode=RCIM_Cubic, leaveTangent=1.0524, tangentMode=RCTM_Break, time=1, value=1}
        ]}}}
     CurveIndex=$Particles.OffsetAge
    )
    Index 3 Param 1=$Particles.FN_Timer.Value
    Module.Param0WriteEnabled=true
    Module.Param0WriteEnabled1=true
    Module.Param0WriteEnabled2=true
    Module.Param0WriteEnabled3=true
    Module.Param1WriteEnabled=true
    Module.Param1WriteEnabled1=true
    Module.Param1WriteEnabled2=true
    Module.Param1WriteEnabled3=true
    Module.Param2WriteEnabled=true
    Module.Param2WriteEnabled1=true
    Module.Param2WriteEnabled2=true
    Module.Param2WriteEnabled3=true
    Module.Param3WriteEnabled=true
    Module.Param3WriteEnabled1=true
    Module.Param3WriteEnabled2=true
    Module.Param3WriteEnabled3=true
    Write Parameter Index 1=true
    Write Parameter Index 2=true
    Write Parameter Index 3=true
   )
   /SetVariables() {enabled=false, isSetParameters=true, module=SetVariables_#6}
   &ScaleColor(
    Scale Alpha=&FloatFromCurve(
     FloatCurve=DataInterface() {propertyValues {Curve {keys [
         {arriveTangent=0.3656, interpMode=RCIM_Cubic, leaveTangent=0.3656, tangentMode=RCTM_User, time=2, value=0}
         {interpMode=RCIM_Cubic, tangentMode=RCTM_User, time=3, value=1}
        ]}}}
     CurveIndex=$Particles.OffsetAge
    )
    ScaleA=true
    ScaleRGB=true
   )
   &ScaleColor(
    Scale Alpha=&MakeCustomFloatFromBool(Bool=$User.IsWhiteVersion, True Float=1.5, False Float=1.5)
    ScaleA=true
    ScaleRGB=true
   ) {module=ScaleColor001}
   /SetVariables(Particles.Position=&Add_Vector(
     A=Linked() {linkedVariable {name=Engine.Owner.Position, type=NiagaraPosition}}
     B=&TransformVector(
      Vector=&MakeVector(
       X=0
       Y=&FloatFromCurve(
        FloatCurve=DataInterface() {propertyValues {Curve {keys [
            {arriveTangent=-0.2987, interpMode=RCIM_Cubic, leaveTangent=-0.2987, tangentMode=RCTM_User, time=0, value=0.75}
            {arriveTangent=0.2159, interpMode=RCIM_Cubic, leaveTangent=0.2159, tangentMode=RCTM_User, time=1, value=0}
           ]}}}
        CurveIndex=&ClampFloat(Float=&Subtract_Float(A=$Particles.Age, B=0.5))
        Scale Curve=-60
       )
       Z=0
      )
      Destination Space=::World
     )
    )) {isSetParameters=true, module=SetVariables_#8}
   /SetVariables(Particles.Angle:NiagaraFloat=/Game/Effects/Niagara/DynamicInputs/Math/FN_RemapDotProduct(
     SystemAxis:Vector3f=$Engine.Owner.SystemYAxis
     ComparisonAxis:Vector3f=&ReturnCameraVector(Camera Property=::NewEnumerator4, Coordinate Space=::World, Origin Position=$Particles.Position)
     SourceMin:NiagaraFloat=0.95
     SourceMax:NiagaraFloat=1
     DestMin:NiagaraFloat=0
     DestMax:NiagaraFloat=1
    )) {isSetParameters=true}
   &ScaleColor(
    Scale Alpha=&Lerp_Float(A=1, B=0, Alpha=&ClampFloat(Float=$Particles.Angle))
    ScaleA=true
    ScaleRGB=true
   ) {module=ScaleColor002}
   &ScaleMeshSize(Scale Factor=&MakeVector(
     X=&FloatFromCurve(FloatCurve=DataInterface() {propertyValues {Curve {keys [<at 0 0>,<at 2 1>]}}}, CurveIndex=$Particles.OffsetAge)
     Y=1
     Z=1
    ))
   &ScaleColor(
    Scale Alpha=&FNTimeOfDayScalar(Night Scale=0.7, Morning Scale=0.3, Evening Scale=0.3)
    ScaleA=true
    ScaleRGB=true
   ) {module=ScaleColor003}
   &ScaleColor(
    Scale RGB=&VectorFromFloat(Value=&MakeCustomFloatFromBool(
      Bool=$User.IsFrontEnd
      True Float=&MakeCustomFloatFromBool(Bool=$User.IsWhiteVersion, True Float=2, False Float=1)
      False Float=1
     ))
    ScaleA=true
    ScaleRGB=true
   ) {module=ScaleColor004}
  ]
  rendererClasses [&NiagaraMeshRendererProperties]
  renderers [
   {rendererClass=&NiagaraMeshRendererProperties, rendererIndex=0}
   {rendererClass=&NiagaraMeshRendererProperties, rendererIndex=1}
   {rendererClass=&NiagaraMeshRendererProperties, rendererIndex=2}
   {rendererClass=&NiagaraMeshRendererProperties, rendererIndex=3}
  ]
 }
 Empty003 {
  emitterUpdateScript [
   &EmitterState(
    Inactive Response=::NewEnumerator1
    Life Cycle Mode=::NewEnumerator1
    Loop Behavior=::NewEnumerator1
    Loop Delay=&MakeCustomFloatFromBool(
     Bool=&SetBoolByIntComparison(A=$User.Deployed, B=0, Comparison Type=::NewEnumerator0)
     True Float=0
     False Float=$User.GrowDelay
    )
    Loop Duration Mode=::NewEnumerator1
    UseLoopDelay=true
   )
   &SpawnBurst_Instantaneous(Use Spawn Probability=false, UseLoopCountLimit=false)
  ]
  particleSpawnScript [
   &InitializeParticle(Color Mode=::NewEnumerator0, Lifetime=9999999, UsePositionOffset=true, Write Material Random=true)
   /SetVariables() {isSetParameters=true, module=SetVariables_#2}
  ]
  particleUpdateScript [
   &ParticleState()
   /SetVariables(Particles.InterpolatePos:Vector3f=&TransformPosition(Position=<0 300 0>, Source Space=::Local, Destination Space=::Simulation)) {isSetParameters=true}
   &InterpolateOverTime(Rate Of Change=2, Target Vector Value:Vector3f=$Particles.InterpolatePos, Variable Type=::NewEnumerator2)
   &self:Clamp(
    In:Vector3f=&Lerp_Vector(
     A=$Particles.InterpolatePos
     B=$Particles.InterpolateOverTime.InterpolateOverTime_V3_.Moving Average
     Alpha=&ClampFloat(Float=&Divide_Float(A=$Particles.Age, B=3))
    )
    Max Angle=75
   ) {module=Clamp}
   /SetVariables(Particles.InterpolateOverTime.InterpolateOverTime_V3_.Moving Average:Vector3f=$Particles.Clamp.Clamped) {isSetParameters=true, module=SetVariables_#3}
   /SetVariables(Particles.Position2:Vector3f=&Add_Vector(
     A=Linked() {linkedVariable {name=Engine.Owner.Position, type=NiagaraPosition}}
     B=&Multiply_VectorByFloat(
      Vector=&NormalizeVector(
       Vector To Normalize=&Subtract_Vector(A=$Particles.Clamp.Clamped, B=Linked() {linkedVariable {name=Engine.Owner.Position, type=NiagaraPosition}})
       Safe Normalize=true
       Safe Fallback Threshold=0
       Safe Fallback Vector=<0 0 0.001>
      )
      Float=300
     )
    )) {isSetParameters=true, module=SetVariables_#4}
   /SetVariables(Particles.MeshOrientation=$Engine.Owner.Rotation) {isSetParameters=true, module=SetVariables_#5}
   /SetVariables(Particles.OffsetAge:NiagaraFloat=&Subtract_Float(
     A=&Add_Float(A=$Particles.Age, B=&Multiply_FloatByInt(Float=100, Integer=$User.Deployed))
     B=1.75
    )) {isSetParameters=true}
   &FN_Timer(Increment Multiplier:NiagaraFloat=&Lerp_Float(A=0.9, B=1.3, Alpha=&ClampFloat(Float=$System.InterpolateOverTime.InterpolateOverTime_Float.Moving Average)))
   &DynamicMaterialParameters(
    Index 0 Param 1=&MakeFloatFromVector(Vector=$Particles.Position2, Channel=::NewEnumerator0)
    Index 0 Param 2=&MakeFloatFromVector(Vector=$Particles.Position2, Channel=::NewEnumerator1)
    Index 0 Param 3=&MakeFloatFromVector(Vector=$Particles.Position2, Channel=::NewEnumerator2)
    Index 0 Param 4=&MakeFloatFromVector(Vector=&TransformVector(Vector=<0 1 0>, Destination Space=::World), Channel=::NewEnumerator0)
    Index 1 Param 1=&MakeFloatFromVector(Vector=&TransformVector(Vector=<0 1 0>, Destination Space=::World), Channel=::NewEnumerator1)
    Index 1 Param 2=&MakeFloatFromVector(Vector=&TransformVector(Vector=<0 1 0>, Destination Space=::World), Channel=::NewEnumerator2)
    Index 1 Param 3=&MakeFloatFromVector(Vector=Linked() {linkedVariable {name=Engine.Owner.Position, type=NiagaraPosition}}, Channel=::NewEnumerator0)
    Index 1 Param 4=&MakeFloatFromVector(Vector=Linked() {linkedVariable {name=Engine.Owner.Position, type=NiagaraPosition}}, Channel=::NewEnumerator1)
    Index 2 Param 1=&MakeFloatFromVector(Vector=Linked() {linkedVariable {name=Engine.Owner.Position, type=NiagaraPosition}}, Channel=::NewEnumerator2)
    Index 2 Param 2=1
    Index 2 Param 4=&FloatFromCurve(
     FloatCurve=DataInterface() {propertyValues {Curve {keys [{interpMode=RCIM_Cubic, tangentMode=RCTM_User, time=0, value=0}, {interpMode=RCIM_Cubic, tangentMode=RCTM_User, time=1, value=1}]}}}
     CurveIndex=$Particles.OffsetAge
    )
    Module.Param0WriteEnabled=true
    Module.Param0WriteEnabled1=true
    Module.Param0WriteEnabled2=true
    Module.Param0WriteEnabled3=true
    Module.Param1WriteEnabled=true
    Module.Param1WriteEnabled1=true
    Module.Param1WriteEnabled2=true
    Module.Param1WriteEnabled3=true
    Module.Param2WriteEnabled=true
    Module.Param2WriteEnabled1=true
    Module.Param2WriteEnabled2=true
    Module.Param2WriteEnabled3=true
    Module.Param3WriteEnabled=true
    Module.Param3WriteEnabled1=true
    Module.Param3WriteEnabled2=true
    Module.Param3WriteEnabled3=true
    Write Parameter Index 1=true
    Write Parameter Index 2=true
    Write Parameter Index 3=true
   )
   /SetVariables() {isSetParameters=true, module=SetVariables_#6}
   /SetVariables(Particles.Position=&Add_Vector(
     A=Linked() {linkedVariable {name=Engine.Owner.Position, type=NiagaraPosition}}
     B=&TransformVector(
      Vector=&MakeVector(
       X=0
       Y=&FloatFromCurve(
        FloatCurve=DataInterface() {propertyValues {Curve {keys [
            {arriveTangent=-0.2987, interpMode=RCIM_Cubic, leaveTangent=-0.2987, tangentMode=RCTM_User, time=0, value=0.75}
            {arriveTangent=0.2159, interpMode=RCIM_Cubic, leaveTangent=0.2159, tangentMode=RCTM_User, time=1, value=0}
           ]}}}
        CurveIndex=$Particles.OffsetAge
        Scale Curve=-10
       )
       Z=0
      )
      Destination Space=::World
     )
    )) {isSetParameters=true}
   &ScaleColor(
    Scale RGB=&VectorFromFloat(Value=&FNTimeOfDayScalar(Night Scale=0.6, Morning Scale=0.3, Evening Scale=0.3))
    ScaleA=true
    ScaleRGB=true
   ) {module=ScaleColor001}
   &ScaleColor(
    Scale Alpha=&MakeCustomFloatFromBool(Bool=$User.IsFrontEnd, True Float=1, False Float=1)
    Scale RGB=&VectorFromFloat(Value=&MakeCustomFloatFromBool(
      Bool=$User.IsFrontEnd
      True Float=&MakeCustomFloatFromBool(Bool=$User.IsWhiteVersion, True Float=1.5, False Float=2)
      False Float=1
     ))
    ScaleA=true
    ScaleRGB=true
   ) {module=ScaleColor002}
  ]
  rendererClasses [&NiagaraMeshRendererProperties]
  renderers [
   {rendererClass=&NiagaraMeshRendererProperties, rendererIndex=0}
   {rendererClass=&NiagaraMeshRendererProperties, rendererIndex=1}
  ]
 }
 Empty004 {
  emitterUpdateScript [
   &EmitterState(
    Inactive Response=::NewEnumerator1
    Life Cycle Mode=::NewEnumerator1
    Loop Behavior=::NewEnumerator1
    Loop Delay=&MakeCustomFloatFromBool(
     Bool=&SetBoolByIntComparison(A=$User.Deployed, B=0, Comparison Type=::NewEnumerator0)
     True Float=0
     False Float=$User.GrowDelay
    )
    Loop Duration Mode=::NewEnumerator1
    UseLoopDelay=true
   )
   &SpawnBurst_Instantaneous(Use Spawn Probability=false, UseLoopCountLimit=false)
  ]
  particleSpawnScript [
   &InitializeParticle(Color Mode=::NewEnumerator0, Lifetime=9999999, UsePositionOffset=true, Write Material Random=true)
   /SetVariables() {isSetParameters=true, module=SetVariables_#2}
  ]
  particleUpdateScript [
   &ParticleState()
   /SetVariables(Particles.InterpolatePos:Vector3f=&TransformPosition(Position=<0 300 0>, Source Space=::Local, Destination Space=::Simulation)) {isSetParameters=true}
   &InterpolateOverTime(Rate Of Change=2, Target Vector Value:Vector3f=$Particles.InterpolatePos, Variable Type=::NewEnumerator2)
   &self:Clamp(
    In:Vector3f=&Lerp_Vector(
     A=$Particles.InterpolatePos
     B=$Particles.InterpolateOverTime.InterpolateOverTime_V3_.Moving Average
     Alpha=&ClampFloat(Float=&Divide_Float(A=$Particles.Age, B=3))
    )
    Max Angle=75
   ) {module=Clamp}
   /SetVariables(Particles.InterpolateOverTime.InterpolateOverTime_V3_.Moving Average:Vector3f=$Particles.Clamp.Clamped) {isSetParameters=true, module=SetVariables_#3}
   /SetVariables(Particles.Position2:Vector3f=&Add_Vector(
     A=Linked() {linkedVariable {name=Engine.Owner.Position, type=NiagaraPosition}}
     B=&Multiply_VectorByFloat(
      Vector=&NormalizeVector(
       Vector To Normalize=&Subtract_Vector(A=$Particles.Clamp.Clamped, B=Linked() {linkedVariable {name=Engine.Owner.Position, type=NiagaraPosition}})
       Safe Normalize=true
       Safe Fallback Threshold=0
       Safe Fallback Vector=<0 0 0.001>
      )
      Float=300
     )
    )) {isSetParameters=true, module=SetVariables_#4}
   /SetVariables(Particles.MeshOrientation=$Engine.Owner.Rotation) {isSetParameters=true, module=SetVariables_#5}
   /SetVariables(Particles.OffsetAge:NiagaraFloat=&Subtract_Float(
     A=&Add_Float(A=$Particles.Age, B=&Multiply_FloatByInt(Float=100, Integer=$User.Deployed))
     B=1.5
    )) {isSetParameters=true}
   &FN_Timer(Increment Multiplier:NiagaraFloat=&FloatFromCurve(
     FloatCurve=DataInterface() {propertyValues {Curve {keys [{interpMode=RCIM_Cubic, tangentMode=RCTM_User, time=0, value=1.5}, {interpMode=RCIM_Cubic, tangentMode=RCTM_User, time=1, value=1}]}}}
     CurveIndex=$Particles.OffsetAge
     Scale Curve=&Lerp_Float(A=0.9, B=1.3, Alpha=&ClampFloat(Float=$System.InterpolateOverTime.InterpolateOverTime_Float.Moving Average))
    ))
   &DynamicMaterialParameters(
    Index 0 Param 1=&MakeFloatFromVector(Vector=$Particles.Position2, Channel=::NewEnumerator0)
    Index 0 Param 2=&MakeFloatFromVector(Vector=$Particles.Position2, Channel=::NewEnumerator1)
    Index 0 Param 3=&MakeFloatFromVector(Vector=$Particles.Position2, Channel=::NewEnumerator2)
    Index 0 Param 4=&MakeFloatFromVector(Vector=&TransformVector(Vector=<0 1 0>, Destination Space=::World), Channel=::NewEnumerator0)
    Index 1 Param 1=&MakeFloatFromVector(Vector=&TransformVector(Vector=<0 1 0>, Destination Space=::World), Channel=::NewEnumerator1)
    Index 1 Param 2=&MakeFloatFromVector(Vector=&TransformVector(Vector=<0 1 0>, Destination Space=::World), Channel=::NewEnumerator2)
    Index 1 Param 3=&MakeFloatFromVector(Vector=Linked() {linkedVariable {name=Engine.Owner.Position, type=NiagaraPosition}}, Channel=::NewEnumerator0)
    Index 1 Param 4=&MakeFloatFromVector(Vector=Linked() {linkedVariable {name=Engine.Owner.Position, type=NiagaraPosition}}, Channel=::NewEnumerator1)
    Index 2 Param 1=&MakeFloatFromVector(Vector=Linked() {linkedVariable {name=Engine.Owner.Position, type=NiagaraPosition}}, Channel=::NewEnumerator2)
    Index 2 Param 2=1
    Index 2 Param 3=&FloatFromCurve(
     FloatCurve=DataInterface() {propertyValues {Curve {keys [{interpMode=RCIM_Cubic, tangentMode=RCTM_User, time=0, value=0}, {interpMode=RCIM_Cubic, tangentMode=RCTM_User, time=1, value=0}]}}}
     CurveIndex=$Particles.OffsetAge
    )
    Index 2 Param 4=&FloatFromCurve(
     FloatCurve=DataInterface() {propertyValues {Curve {keys [
         {arriveTangent=0.9949, interpMode=RCIM_Cubic, leaveTangent=0.9949, tangentMode=RCTM_User, time=0, value=0}
         {arriveTangent=0.8191, interpMode=RCIM_Cubic, leaveTangent=0.8191, tangentMode=RCTM_User, time=1, value=1}
        ]}}}
     CurveIndex=$Particles.OffsetAge
    )
    Index 3 Param 1=&Add_Float(A=$Particles.FN_Timer.Value, B=1.1758)
    Module.Param0WriteEnabled=true
    Module.Param0WriteEnabled1=true
    Module.Param0WriteEnabled2=true
    Module.Param0WriteEnabled3=true
    Module.Param1WriteEnabled=true
    Module.Param1WriteEnabled1=true
    Module.Param1WriteEnabled2=true
    Module.Param1WriteEnabled3=true
    Module.Param2WriteEnabled=true
    Module.Param2WriteEnabled1=true
    Module.Param2WriteEnabled2=true
    Module.Param2WriteEnabled3=true
    Module.Param3WriteEnabled=true
    Module.Param3WriteEnabled1=true
    Module.Param3WriteEnabled2=true
    Module.Param3WriteEnabled3=true
    Write Parameter Index 1=true
    Write Parameter Index 2=true
    Write Parameter Index 3=true
   )
   /SetVariables() {isSetParameters=true, module=SetVariables_#6}
   /SetVariables(Particles.Position=&Add_Vector(
     A=Linked() {linkedVariable {name=Engine.Owner.Position, type=NiagaraPosition}}
     B=&TransformVector(
      Vector=&MakeVector(
       X=0
       Y=&FloatFromCurve(
        FloatCurve=DataInterface() {propertyValues {Curve {keys [
            {arriveTangent=-0.2987, interpMode=RCIM_Cubic, leaveTangent=-0.2987, tangentMode=RCTM_User, time=0, value=0.75}
            {arriveTangent=0.2159, interpMode=RCIM_Cubic, leaveTangent=0.2159, tangentMode=RCTM_User, time=1, value=0}
           ]}}}
        CurveIndex=$Particles.OffsetAge
        Scale Curve=-60
       )
       Z=0
      )
      Destination Space=::World
     )
    )) {isSetParameters=true}
   &ScaleColor(
    Scale RGB=&VectorFromFloat(Value=&FNTimeOfDayScalar(Night Scale=0.6, Morning Scale=0.4, Evening Scale=0.4))
    ScaleA=true
    ScaleRGB=true
   ) {module=ScaleColor001}
   &ScaleColor(
    Scale Alpha=&MakeCustomFloatFromBool(Bool=$User.IsFrontEnd, True Float=0.7, False Float=1)
    Scale RGB=&VectorFromFloat(Value=&MakeCustomFloatFromBool(
      Bool=$User.IsFrontEnd
      True Float=&MakeCustomFloatFromBool(Bool=$User.IsWhiteVersion, True Float=1.5, False Float=2)
      False Float=1
     ))
    ScaleA=true
    ScaleRGB=true
   ) {module=ScaleColor002}
  ]
  rendererClasses [&NiagaraMeshRendererProperties]
  renderers [
   {rendererClass=&NiagaraMeshRendererProperties, rendererIndex=0}
   {rendererClass=&NiagaraMeshRendererProperties, rendererIndex=1}
  ]
 }
 Empty005 {
  emitterUpdateScript [
   &EmitterState(
    Inactive Response=::NewEnumerator1
    Life Cycle Mode=::NewEnumerator1
    Loop Behavior=::NewEnumerator1
    Loop Delay=0.15
    Loop Duration Mode=::NewEnumerator1
    UseLoopDelay=true
   )
   &SpawnBurst_Instantaneous(Use Spawn Probability=false, UseLoopCountLimit=false)
  ]
  particleSpawnScript [&InitializeParticle(
    Color Mode=::NewEnumerator0
    Lifetime=1.7
    Mesh Scale Mode=::NewEnumerator3
    Mesh Uniform Scale=0.05
    UsePositionOffset=true
    Write Material Random=true
   )]
  particleUpdateScript [
   &ParticleState()
   &UpdateMeshOrientation(Rotation Coordinate Space=::NewEnumerator2, Rotation Rate=1.5, Rotation Vector=<-1 0 0>)
   &ScaleMeshSize(Scale Factor=&VectorFromFloat(Value=&FloatFromCurve(
      FloatCurve=DataInterface() {propertyValues {Curve {keys [<at 0.03 0>,<at 0.1 1>,<at 1.5 1>,<at 1.7 0>]}}}
      CurveIndex=$Particles.Age
      Scale Curve=&MakeCustomFloatFromBool(Bool=$User.IsFrontEnd, True Float=0.7, False Float=1)
     )))
   &ScaleColor(
    Scale RGB=&VectorFromFloat(Value=&FNTimeOfDayScalar(Night Scale=1, Morning Scale=0.2, Evening Scale=0.2))
    ScaleA=true
    ScaleRGB=true
   )
  ]
  rendererClasses [&NiagaraMeshRendererProperties]
  renderers [{rendererClass=&NiagaraMeshRendererProperties, rendererIndex=0}]
 }
 Empty006 {
  emitterUpdateScript [
   &EmitterState(
    Inactive Response=::NewEnumerator1
    Life Cycle Mode=::NewEnumerator1
    Loop Behavior=::NewEnumerator1
    Loop Delay=&MakeCustomFloatFromBool(
     Bool=&SetBoolByIntComparison(A=$User.Deployed, B=0, Comparison Type=::NewEnumerator0)
     True Float=0
     False Float=$User.GrowDelay
    )
    Loop Duration Mode=::NewEnumerator1
    UseLoopDelay=true
   )
   &SpawnBurst_Instantaneous(Use Spawn Probability=false, UseLoopCountLimit=false)
  ]
  particleSpawnScript [
   &InitializeParticle(Color Mode=::NewEnumerator0, Lifetime=9999999, Mesh Scale Mode=::NewEnumerator3, UsePositionOffset=true, Write Material Random=true)
   /SetVariables() {isSetParameters=true, module=SetVariables_#2}
  ]
  particleUpdateScript [
   &ParticleState()
   /SetVariables(Particles.InterpolatePos:Vector3f=&TransformPosition(Position=<0 300 0>, Source Space=::Local, Destination Space=::Simulation)) {isSetParameters=true}
   &InterpolateOverTime(Rate Of Change=2, Target Vector Value:Vector3f=$Particles.InterpolatePos, Variable Type=::NewEnumerator2)
   &self:Clamp(
    In:Vector3f=&Lerp_Vector(
     A=$Particles.InterpolatePos
     B=$Particles.InterpolateOverTime.InterpolateOverTime_V3_.Moving Average
     Alpha=&ClampFloat(Float=&Divide_Float(A=$Particles.Age, B=3))
    )
    Max Angle=75
   ) {module=Clamp}
   /SetVariables(Particles.InterpolateOverTime.InterpolateOverTime_V3_.Moving Average:Vector3f=$Particles.Clamp.Clamped) {isSetParameters=true, module=SetVariables_#3}
   /SetVariables(Particles.Position2:Vector3f=&Add_Vector(
     A=Linked() {linkedVariable {name=Engine.Owner.Position, type=NiagaraPosition}}
     B=&Multiply_VectorByFloat(
      Vector=&NormalizeVector(
       Vector To Normalize=&Subtract_Vector(A=$Particles.Clamp.Clamped, B=Linked() {linkedVariable {name=Engine.Owner.Position, type=NiagaraPosition}})
       Safe Normalize=true
       Safe Fallback Threshold=0
       Safe Fallback Vector=<0 0 0.001>
      )
      Float=300
     )
    )) {isSetParameters=true, module=SetVariables_#4}
   /SetVariables(Particles.MeshOrientation=$Engine.Owner.Rotation) {isSetParameters=true, module=SetVariables_#5}
   /SetVariables(Particles.OffsetAge:NiagaraFloat=&Subtract_Float(
     A=&Add_Float(A=$Particles.Age, B=&Multiply_FloatByInt(Float=100, Integer=$User.Deployed))
     B=0.5
    )) {isSetParameters=true, module=SetVariables_#7}
   &FN_Timer(Increment Multiplier:NiagaraFloat=&FloatFromCurve(
     FloatCurve=DataInterface() {propertyValues {Curve {keys [{interpMode=RCIM_Cubic, tangentMode=RCTM_User, time=0, value=3}, {interpMode=RCIM_Cubic, tangentMode=RCTM_User, time=2, value=1}]}}}
     CurveIndex=$Particles.OffsetAge
     Scale Curve=&Lerp_Float(A=0.9, B=1.3, Alpha=&ClampFloat(Float=$System.InterpolateOverTime.InterpolateOverTime_Float.Moving Average))
    ))
   &DynamicMaterialParameters(
    Index 0 Param 1=&MakeFloatFromVector(Vector=$Particles.Position2, Channel=::NewEnumerator0)
    Index 0 Param 2=&MakeFloatFromVector(Vector=$Particles.Position2, Channel=::NewEnumerator1)
    Index 0 Param 3=&MakeFloatFromVector(Vector=$Particles.Position2, Channel=::NewEnumerator2)
    Index 0 Param 4=&MakeFloatFromVector(Vector=&TransformVector(Vector=<0 1 0>, Destination Space=::World), Channel=::NewEnumerator0)
    Index 1 Param 1=&MakeFloatFromVector(Vector=&TransformVector(Vector=<0 1 0>, Destination Space=::World), Channel=::NewEnumerator1)
    Index 1 Param 2=&MakeFloatFromVector(Vector=&TransformVector(Vector=<0 1 0>, Destination Space=::World), Channel=::NewEnumerator2)
    Index 1 Param 3=&MakeFloatFromVector(Vector=Linked() {linkedVariable {name=Engine.Owner.Position, type=NiagaraPosition}}, Channel=::NewEnumerator0)
    Index 1 Param 4=&MakeFloatFromVector(Vector=Linked() {linkedVariable {name=Engine.Owner.Position, type=NiagaraPosition}}, Channel=::NewEnumerator1)
    Index 2 Param 1=&MakeFloatFromVector(Vector=Linked() {linkedVariable {name=Engine.Owner.Position, type=NiagaraPosition}}, Channel=::NewEnumerator2)
    Index 2 Param 2=1
    Index 2 Param 4=&FloatFromCurve(
     FloatCurve=DataInterface() {propertyValues {Curve {keys [
         {arriveTangent=1.3853, interpMode=RCIM_Cubic, leaveTangent=1.3853, tangentMode=RCTM_User, time=0, value=0}
         {arriveTangent=2.125, interpMode=RCIM_Cubic, leaveTangent=1.0524, tangentMode=RCTM_Break, time=1, value=1}
        ]}}}
     CurveIndex=$Particles.OffsetAge
    )
    Index 3 Param 1=$Particles.FN_Timer.Value
    Module.Param0WriteEnabled=true
    Module.Param0WriteEnabled1=true
    Module.Param0WriteEnabled2=true
    Module.Param0WriteEnabled3=true
    Module.Param1WriteEnabled=true
    Module.Param1WriteEnabled1=true
    Module.Param1WriteEnabled2=true
    Module.Param1WriteEnabled3=true
    Module.Param2WriteEnabled=true
    Module.Param2WriteEnabled1=true
    Module.Param2WriteEnabled2=true
    Module.Param2WriteEnabled3=true
    Module.Param3WriteEnabled=true
    Module.Param3WriteEnabled1=true
    Module.Param3WriteEnabled2=true
    Module.Param3WriteEnabled3=true
    Write Parameter Index 1=true
    Write Parameter Index 2=true
    Write Parameter Index 3=true
   )
   /SetVariables() {enabled=false, isSetParameters=true, module=SetVariables_#6}
   &ScaleColor(
    Scale Alpha=&FloatFromCurve(
     FloatCurve=DataInterface() {propertyValues {Curve {keys [
         {arriveTangent=0.4681, interpMode=RCIM_Cubic, leaveTangent=0.4681, tangentMode=RCTM_User, time=1.8, value=0}
         {arriveTangent=0.005, interpMode=RCIM_Cubic, leaveTangent=0.005, tangentMode=RCTM_User, time=2.5, value=1}
         {interpMode=RCIM_Cubic, tangentMode=RCTM_User, time=3, value=1}
        ]}}}
     CurveIndex=$Particles.OffsetAge
    )
    ScaleA=true
    ScaleRGB=true
   )
   &ScaleColor(ScaleA=true, ScaleRGB=true) {module=ScaleColor001}
   /SetVariables(Particles.Position=&Add_Vector(
     A=Linked() {linkedVariable {name=Engine.Owner.Position, type=NiagaraPosition}}
     B=&TransformVector(
      Vector=&MakeVector(
       X=0
       Y=&FloatFromCurve(
        FloatCurve=DataInterface() {propertyValues {Curve {keys [
            {arriveTangent=-0.2987, interpMode=RCIM_Cubic, leaveTangent=-0.2987, tangentMode=RCTM_User, time=0, value=0.75}
            {arriveTangent=0.2159, interpMode=RCIM_Cubic, leaveTangent=0.2159, tangentMode=RCTM_User, time=1, value=0}
           ]}}}
        CurveIndex=&ClampFloat(Float=&Subtract_Float(A=$Particles.Age, B=0.5))
        Scale Curve=-60
       )
       Z=0
      )
      Destination Space=::World
     )
    )) {isSetParameters=true, module=SetVariables_#8}
   /SetVariables(Particles.Angle:NiagaraFloat=&FN_RemapDotProduct(
     SystemAxis:Vector3f=$Engine.Owner.SystemYAxis
     ComparisonAxis:Vector3f=&ReturnCameraVector(Camera Property=::NewEnumerator4, Coordinate Space=::World, Origin Position=$Particles.Position)
     SourceMin:NiagaraFloat=0.95
     SourceMax:NiagaraFloat=1
     DestMin:NiagaraFloat=0
     DestMax:NiagaraFloat=1
    )) {isSetParameters=true}
   &ScaleColor(
    Scale Alpha=&Lerp_Float(A=1, B=0, Alpha=&ClampFloat(Float=$Particles.Angle))
    ScaleA=true
    ScaleRGB=true
   ) {module=ScaleColor002}
   &ScaleColor(
    Scale Alpha=&FNTimeOfDayScalar(Night Scale=1, Morning Scale=0.3, Evening Scale=0.3)
    ScaleA=true
    ScaleRGB=true
   ) {module=ScaleColor003}
   &ScaleColor(
    Scale RGB=&VectorFromFloat(Value=&MakeCustomFloatFromBool(
      Bool=$User.IsFrontEnd
      True Float=&MakeCustomFloatFromBool(Bool=$User.IsWhiteVersion, True Float=1.25, False Float=1.5)
      False Float=1
     ))
    ScaleA=true
    ScaleRGB=true
   ) {module=ScaleColor004}
  ]
  rendererClasses [&NiagaraMeshRendererProperties]
  renderers [
   {rendererClass=&NiagaraMeshRendererProperties, rendererIndex=0}
   {rendererClass=&NiagaraMeshRendererProperties, rendererIndex=1}
  ]
 }
 Tennis_ball_particle {
  emitterUpdateScript [
   &EmitterState(
    Inactive Response=::NewEnumerator1
    Life Cycle Mode=::NewEnumerator1
    Loop Behavior=::NewEnumerator1
    Loop Delay=&MakeCustomFloatFromBool(
     Bool=&SetBoolByIntComparison(A=$User.Deployed, B=0, Comparison Type=::NewEnumerator0)
     True Float=0
     False Float=1.4
    )
    Loop Duration=35790.3945
    UseLoopDelay=true
   )
   &SpawnRate(
    SpawnRate=&MakeCustomFloatFromBool(Bool=$User.IsWhiteVersion, True Float=2, False Float=0)
    Use Spawn Probability=false
   )
  ]
  particleSpawnScript [
   &InitializeParticle(
    Lifetime Max=3
    Lifetime Min=2
    Lifetime Mode=::NewEnumerator1
    Mass Mode=::NewEnumerator3
    Mesh Scale Mode=::NewEnumerator1
    Mesh Uniform Scale Max=0.045
    Mesh Uniform Scale Min=0.025
    Sprite Rotation Mode=::NewEnumerator2
    Sprite Size Mode=::NewEnumerator1
    Uniform Sprite Size Max=12
    Uniform Sprite Size Min=10
    UsePositionOffset=false
    Write Material Random=true
   )
   &SphereLocation(
    MaskSpawnGroup=false
    Offset=<0 -115.5987 40>
    Override Seed=false
    Sphere Radius=10
    Surface Only=false
    UseNonUniformScale=true
    UseOffset=true
    UseSphereOrientation=true
   )
   &AddVelocityInCone(
    Cone Angle=159.8498
    Cone Axis=<0 1 0>
    Module.Use Velocity Falloff On Cone Axis=false
    Velocity Distribution Along Cone Axis=0
    Velocity Strength=&RandomRangeFloat(Minimum=87.825, Maximum=94.2775)
   )
   &InitialMeshOrientation(Use Rotation Vector=false)
   &AddVelocity()
   /SetVariables(Particles.Velocity=&Multiply_Vector(A=$Particles.Velocity, B=<0.5 1 0.5>)) {isSetParameters=true}
  ]
  particleUpdateScript [
   &ParticleState()
   &self:ReadOtherParticleAttri(New Particle Attribute Reader=DataInterface() {propertyValues {EmitterBinding {bindingMode=Other, emitterName=Empty003}}}) {module=ReadOtherParticleAttri}
   &LineAttractionForce(
    Attraction Strength=150
    Line End:Vector3f=<0 500 0>
    Line Segment in Localspace=true
    Line Start:Vector3f=<0 0 0>
    Use Attraction Falloff=false
   )
   &LinearForce(Force=&Lerp_Vector(
     A=&Lerp_Vector(
      A=<500 0 0>
      B=<-500 0 0>
      Alpha=&ClampFloat(Float=&DotProduct(
        Vector A=&NormalizeVector(
         Vector To Normalize=&Subtract_Vector(A=$Particles.Value1, B=<0 0 0>)
         Safe Normalize=true
         Safe Fallback Threshold=0
         Safe Fallback Vector=<0 0 0.001>
        )
        Vector B=<-1 0 0>
       ))
     )
     B=<0 0 0>
     Alpha=&DotProduct(
      Vector A=&NormalizeVector(
       Vector To Normalize=&Subtract_Vector(A=$Particles.Value1, B=<0 0 0>)
       Safe Normalize=true
       Safe Fallback Threshold=0
       Safe Fallback Vector=<0 0 0.001>
      )
      Vector B=<0 1 0>
     )
    ))
   &CurlNoiseForce(Noise Frequency=10, Noise Strength=50, Pan Noise=false)
   &GravityForce(Gravity=<0 0 20>)
   &Drag(Drag=0.5, Use Linear Drag=true, Use Rotational Drag=false)
   &ScaleColor(
    Scale Alpha=&FloatFromCurve(FloatCurve=DataInterface() {propertyValues {Curve {keys [<at -0.0046 1.0113 RCIM_Cubic>,<at 1.0011 0.002 RCIM_Cubic>]}}})
    Scale RGB=&VectorFromFloat(Value=&FNTimeOfDayScalar(Night Scale=1, Morning Scale=0.25, Evening Scale=0.5))
    ScaleA=true
    ScaleRGB=true
   )
   &PointAttractionForce(
    Attraction Radius=25493.1113
    Attraction Strength=20
    Attractor Position Offset:Vector3f=<0 475.3237 81.9046>
    Falloff Exponent=0.4
    Kill Within Radius=false
    Module.Use Falloff=true
    UseAttractorOffset=true
   )
   &SolveForcesAndVelocity(Clamp Velocity=false, Limit Acceleration=false)
   &ScaleMeshSize(Scale Factor=&VectorFromFloat(Value=&FloatFromCurve(
      FloatCurve=DataInterface() {propertyValues {Curve {keys [
          {arriveTangent=30.2993, interpMode=RCIM_Cubic, leaveTangent=30.2993, tangentMode=RCTM_User, time=0.0989, value=0}
          {arriveTangent=0.0288, interpMode=RCIM_Cubic, leaveTangent=0.0288, tangentMode=RCTM_User, time=0.1897, value=0.9847}
          {arriveTangent=-0.1393, interpMode=RCIM_Cubic, leaveTangent=-0.1393, tangentMode=RCTM_User, time=0.7724, value=0.912}
          {arriveTangent=-9.5947, interpMode=RCIM_Cubic, leaveTangent=-9.5947, tangentMode=RCTM_User, time=1, value=0}
         ]}}}
      Scale Curve=1.25
     )))
   &UpdateMeshOrientation(Rotation Vector=&RandomRangeVector(Minimum=<0 0 0>, Maximum=<1 1 1>))
   &KillParticlesInVolume(Box Size:Vector3f=<25 16 40>, Kill Shape=::NewEnumerator1, Origin Offset:Vector3f=<0 -100 40>) {enabled=false}
   &KillParticlesInVolume(Box Size:Vector3f=<30 30 20>, Kill Shape=::NewEnumerator1, Origin Offset:Vector3f=<0 -100 -40>) {enabled=false, module=KillParticlesInVolume001}
  ]
  rendererClasses [&NiagaraMeshRendererProperties]
  renderers [{rendererClass=&NiagaraMeshRendererProperties, rendererIndex=0}]
 }
}
summary {
 systemName=FX_ThriveParticles
 emitters [
  {emitterName=Tennis_ball_particle, rendererClasses [&NiagaraMeshRendererProperties]}
  {emitterName=Empty, rendererClasses [&NiagaraMeshRendererProperties]}
  {emitterName=Empty001, rendererClasses [&NiagaraMeshRendererProperties]}
  {emitterName=Empty002, rendererClasses [&NiagaraMeshRendererProperties]}
  {emitterName=Empty003, rendererClasses [&NiagaraMeshRendererProperties]}
  {emitterName=Empty004, rendererClasses [&NiagaraMeshRendererProperties]}
  {emitterName=Empty005, rendererClasses [&NiagaraMeshRendererProperties]}
  {emitterName=Empty006, rendererClasses [&NiagaraMeshRendererProperties]}
 ]
 userVariables [
  User.Deployed {type=NiagaraInt32}
  User.ForwardDot {type=NiagaraFloat, defaultValue=0, description=""}
  User.GrowDelay {type=NiagaraFloat, defaultValue=0.4, description=""}
  User.IsFrontEnd {type=NiagaraBool}
  User.IsWhiteVersion {type=NiagaraBool, defaultValue=true}
 ]
}
