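# Niagara effect for the Thrive Female glider: three spirit skulls roll around tilted orbits beside three stationary, gently swaying skulls.
# RollingSkull positions particles with per-group transform matrices and passes the group index through Mass, separating FollowingRibbon's trails by ribbon ID.
# Ribbons fade out toward the owner's height, while Empty002–004 place the tintable static skulls at offsets.
# Sines drive wobble and flicker; IsFrontEnd and time of day adjust opacity.
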NiagaraSystem /BRCosmetics/Gadgets/Assets/VinderTech_GliderChute/Glider_Thrive_Female/FX/FX_Thrive_SpiritRoll
dependencies {
 usedDataInterfaces [
  &NiagaraDataInterfaceArrayFloat
  &NiagaraDataInterfaceArrayFloat3
  &NiagaraDataInterfaceColorCurve
  &NiagaraDataInterfaceParticleRead
  &NiagaraDataInterfaceCurve
  &NiagaraDataInterfaceCamera
 ]
 usedDynamicInputs [&SetBoolByIntComparison]
 usedModules [
  &SystemState
  &self:SystemSpawnScript.SystemScriptSource.SystemScriptGraph.NiagaraNodeAssignment_0.SetVariables_#1
  &self:CameraDirctionFade
 ]
}
emitters {
 Empty002 {
  emitterUpdateScript [
   &EmitterState()
   &SpawnBurst_Instantaneous(Use Spawn Probability=false, UseLoopCountLimit=false)
  ]
  particleSpawnScript [
   &InitializeParticle(
    Color=$User.StaticSkullsColor
    Lifetime=5
    Mesh Scale Mode=::NewEnumerator3
    Mesh Uniform Scale=1.3
    Position Offset=&Add_Vector(A=<0 -89.3996 -50.4086>, B=$User.SkullsOffset)
    UsePositionOffset=true
    Write Material Random=true
   )
   &InitialMeshOrientation(Mesh Orientation Mode=::NewEnumerator2, Orientation Vector=<0 -1 -0.5674>, Use Rotation Vector=false)
  ]
  particleUpdateScript [
   &ParticleState(Kill Particles When Lifetime Has Elapsed=false, Let Infinitely Lived Particles Die When Emitter Deactivates=true)
   &DynamicMaterialParameters(
    Index 0 Param 1=&Sine(
     Normalized Angle=&Add_Float(A=$Emitter.Age, B=&RandomRangeFloat(Minimum=0, Maximum=50))
     Period=3.5
     Scale=0.2
     Bias=0.8
    )
    Module.Param0WriteEnabled=true
    Module.Param1WriteEnabled=true
    Module.Param2WriteEnabled=true
    Module.Param3WriteEnabled=true
   )
   /SetVariables(Particles.RandomRotate:Quat4f=&MakeQuatFromAxisAngle(
     Axis=&Lerp_Vector(
      A=<0 0.5 1>
      B=<0.5 0 1>
      Alpha=&Sine(
       Normalized Angle=&Add_Float(A=&Add_Float(A=$Emitter.Age, B=0), B=-0.2)
       Period=6
       Scale=1
       Bias=0
      )
     )
     AngleInDegrees=&Lerp_Float(
      A=0
      B=30
      Alpha=&Sine(
       Normalized Angle=&Add_Float(A=&Add_Float(A=$Emitter.Age, B=0), B=-0.2)
       Period=4.5
       Scale=1
       Bias=0
      )
     )
     NormalizeAxis=true
    )) {isSetParameters=true}
   /SetVariables(Particles.MeshOrientation2:Quat4f=&MultiplyQuaternion(Quaternion A=$Particles.RandomRotate, Quaternion B=$Particles.MeshOrientation)) {isSetParameters=true}
   &ScaleColor(
    Scale Alpha=&FloatFromCurve(FloatCurve=DataInterface() {propertyValues {Curve {keys [<at 0 0>,<at 0.5 1>]}}}, CurveIndex=$Particles.Age)
    ScaleA=true
    ScaleRGB=true
   ) {module=ScaleColor001}
   &ScaleColor(Scale Alpha=$System.CDF, ScaleA=true, ScaleRGB=true) {enabled=false}
   &ScaleColor(
    Scale Alpha=&MakeCustomFloatFromBool(Bool=$User.IsFrontEnd, True Float=1, False Float=1)
    ScaleA=true
    ScaleRGB=true
   ) {module=ScaleColor002}
  ]
  rendererClasses [&NiagaraMeshRendererProperties]
  renderers [{rendererClass=&NiagaraMeshRendererProperties, rendererIndex=0}]
 }
 Empty003 {
  emitterUpdateScript [
   &EmitterState()
   &SpawnBurst_Instantaneous(Use Spawn Probability=false, UseLoopCountLimit=false)
  ]
  particleSpawnScript [
   &InitializeParticle(
    Color=$User.StaticSkullsColor
    Lifetime=5
    Mesh Scale Mode=::NewEnumerator3
    Mesh Uniform Scale=1.3
    Position Offset=&Add_Vector(A=<-76.3245 -51.2158 -6.223>, B=$User.SkullsOffset)
    UsePositionOffset=true
    Write Material Random=true
   )
   &InitialMeshOrientation(Mesh Orientation Mode=::NewEnumerator2, Orientation Vector=<-5.9455 -8.4348 -0.1659>, Use Rotation Vector=false)
  ]
  particleUpdateScript [
   &ParticleState(Kill Particles When Lifetime Has Elapsed=false, Let Infinitely Lived Particles Die When Emitter Deactivates=true)
   &DynamicMaterialParameters(
    Index 0 Param 1=&Sine(
     Normalized Angle=&Add_Float(A=$Emitter.Age, B=&RandomRangeFloat(Minimum=0, Maximum=50))
     Period=3.5
     Scale=0.2
     Bias=0.8
    )
    Module.Param0WriteEnabled=true
    Module.Param1WriteEnabled=true
    Module.Param2WriteEnabled=true
    Module.Param3WriteEnabled=true
   )
   /SetVariables(Particles.RandomRotate:Quat4f=&MakeQuatFromAxisAngle(
     Axis=&Lerp_Vector(
      A=<0 0.5 1>
      B=<0.5 0 1>
      Alpha=&Sine(
       Normalized Angle=&Add_Float(A=&Add_Float(A=$Emitter.Age, B=0.513), B=-0.2)
       Period=6
       Scale=1
       Bias=0
      )
     )
     AngleInDegrees=&Lerp_Float(
      A=0
      B=30
      Alpha=&Sine(
       Normalized Angle=&Add_Float(A=&Add_Float(A=$Emitter.Age, B=0.513), B=-0.2)
       Period=4.5
       Scale=1
       Bias=0
      )
     )
     NormalizeAxis=true
    )) {isSetParameters=true}
   /SetVariables(Particles.MeshOrientation2:Quat4f=&MultiplyQuaternion(Quaternion A=$Particles.RandomRotate, Quaternion B=$Particles.MeshOrientation)) {isSetParameters=true}
   &ScaleColor(
    Scale Alpha=&FloatFromCurve(FloatCurve=DataInterface() {propertyValues {Curve {keys [<at 0 0>,<at 0.5 1>]}}}, CurveIndex=$Particles.Age)
    ScaleA=true
    ScaleRGB=true
   ) {module=ScaleColor001}
   &ScaleColor(Scale Alpha=$System.CDF, ScaleA=true, ScaleRGB=true) {enabled=false}
   &ScaleColor(
    Scale Alpha=&MakeCustomFloatFromBool(Bool=$User.IsFrontEnd, True Float=4, False Float=1)
    ScaleA=true
    ScaleRGB=true
   ) {module=ScaleColor002}
  ]
  rendererClasses [&NiagaraMeshRendererProperties]
  renderers [{rendererClass=&NiagaraMeshRendererProperties, rendererIndex=0}]
 }
 Empty004 {
  emitterUpdateScript [
   &EmitterState()
   &SpawnBurst_Instantaneous(Use Spawn Probability=false, UseLoopCountLimit=false)
  ]
  particleSpawnScript [
   &InitializeParticle(
    Color=$User.StaticSkullsColor
    Lifetime=5
    Mesh Scale Mode=::NewEnumerator3
    Mesh Uniform Scale=1.3
    Position Offset=&Add_Vector(A=<65.4495 -72.9343 -31.537>, B=$User.SkullsOffset)
    UsePositionOffset=true
    Write Material Random=true
   )
   &InitialMeshOrientation(Mesh Orientation Mode=::NewEnumerator2, Orientation Vector=<0 -1 -0.29>, Use Rotation Vector=false)
  ]
  particleUpdateScript [
   &ParticleState(Kill Particles When Lifetime Has Elapsed=false, Let Infinitely Lived Particles Die When Emitter Deactivates=true)
   &DynamicMaterialParameters(
    Index 0 Param 1=&Sine(
     Normalized Angle=&Add_Float(A=$Emitter.Age, B=&RandomRangeFloat(Minimum=0, Maximum=50))
     Period=3.5
     Scale=0.2
     Bias=0.8
    )
    Module.Param0WriteEnabled=true
    Module.Param1WriteEnabled=true
    Module.Param2WriteEnabled=true
    Module.Param3WriteEnabled=true
   )
   /SetVariables(Particles.RandomRotate:Quat4f=&MakeQuatFromAxisAngle(
     Axis=&Lerp_Vector(
      A=<0 0.5 1>
      B=<0.5 0 1>
      Alpha=&Sine(
       Normalized Angle=&Add_Float(A=&Add_Float(A=$Emitter.Age, B=52.5153), B=-0.2)
       Period=6
       Scale=1
       Bias=0
      )
     )
     AngleInDegrees=&Lerp_Float(
      A=0
      B=30
      Alpha=&Sine(
       Normalized Angle=&Add_Float(A=&Add_Float(A=$Emitter.Age, B=52.5153), B=-0.2)
       Period=4.5
       Scale=1
       Bias=0
      )
     )
     NormalizeAxis=true
    )) {isSetParameters=true}
   /SetVariables(Particles.MeshOrientation2:Quat4f=&MultiplyQuaternion(Quaternion A=$Particles.RandomRotate, Quaternion B=$Particles.MeshOrientation)) {isSetParameters=true}
   &ScaleColor(
    Scale Alpha=&FloatFromCurve(FloatCurve=DataInterface() {propertyValues {Curve {keys [<at 0 0>,<at 0.5 1>]}}}, CurveIndex=$Particles.Age)
    ScaleA=true
    ScaleRGB=true
   ) {module=ScaleColor001}
   &ScaleColor(Scale Alpha=$System.CDF, ScaleA=true, ScaleRGB=true) {enabled=false}
   &ScaleColor(
    Scale Alpha=&MakeCustomFloatFromBool(Bool=$User.IsFrontEnd, True Float=4, False Float=1)
    ScaleA=true
    ScaleRGB=true
   ) {module=ScaleColor002}
  ]
  rendererClasses [&NiagaraMeshRendererProperties]
  renderers [{rendererClass=&NiagaraMeshRendererProperties, rendererIndex=0}]
 }
 FollowingRibbon {
  emitterUpdateScript [
   &EmitterState()
   &SpawnParticlesFromOtherEmitter(
    Attribute Reader=DataInterface() {propertyValues {EmitterBinding {bindingMode=Other, emitterName=RollingSkull}}}
    Calculate Spawn Rate Per Particle=false
    Spawn Rate=75
    Use Maximum Age=false
    Use Minimum Age=false
   )
  ]
  particleSpawnScript [
   &InitializeParticle(Color=<rgba 1 1 1 500>, Lifetime=15, UsePositionOffset=false, Write Material Random=true)
   &InitializeRibbon(Lifetime=1.5, Ribbon Width=50, UsePositionOffset=false, Write Material Random=true)
   &SampleParticlesFromOtherEmitter(Mass Application:ENiagara_LifetimeInheritanceOptions=::NewEnumerator2, Mass Sampling:ENiagara_DirectReadSamplingMode=::NewEnumerator1)
   /SetVariables(Particles.RibbonID=&self:ScratchDynamicInput(SpawnGroup:NiagaraInt32=&MakeIntFromFloat(Float=$Output.SampleParticlesFromOtherEmitter.SampledMass))) {isSetParameters=true}
  ]
  particleUpdateScript [
   &ParticleState()
   &ScaleColor(
    Scale Alpha=&ClampFloat(Float=&Divide_Float(
      A=&Subtract_Float(
       A=&MakeFloatFromVector(
        Vector=&Subtract_Vector(
         A=/Niagara/DynamicInputs/Transforms/TransformPosition(Position=Linked() {linkedVariable {name=Particles.Position, type=NiagaraPosition}}, Source Space=::Local, Destination Space=::World)
         B=Linked() {linkedVariable {name=Engine.Owner.Position, type=NiagaraPosition}}
        )
        Channel=::NewEnumerator2
       )
       B=0
      )
      B=32
     ))
    ScaleA=true
    ScaleRGB=true
   )
   &ScaleColor(
    Scale Alpha=&MakeCustomFloatFromBool(Bool=$User.IsFrontEnd, True Float=1.5, False Float=1)
    ScaleA=true
    ScaleRGB=true
   ) {module=ScaleColor002}
  ]
  rendererClasses [&NiagaraRibbonRendererProperties]
  renderers [{rendererClass=&NiagaraRibbonRendererProperties}]
 }
 MeshOrientation {
  bEnabled=false
  emitterUpdateScript [
   &EmitterState(Life Cycle Mode=::NewEnumerator1, Loop Duration=1, UseLoopDelay=false)
   &SpawnBurst_Instantaneous(Use Spawn Probability=false, UseLoopCountLimit=false)
  ]
  particleSpawnScript [
   &InitializeParticle(Lifetime=3, Mesh Scale Mode=::NewEnumerator3, Mesh Uniform Scale=20, UsePositionOffset=false, Write Material Random=true)
   &InitialMeshOrientation(Orientation Vector=&RandomVector(Vector Scale=1), Use Orientation Vector=true, Use Rotation Vector=false) {enabled=false}
  ]
  particleUpdateScript [
   &ParticleState()
   &UpdateMeshOrientation(Rotation Coordinate Space=::NewEnumerator3, Rotation Vector=<0 0 0.5>) {enabled=false}
   /Niagara/Modules/Update/Orientation/V2/OrientMeshToVector(Override Quat=false, UseOptionalSideAxis=false, UseOptionalUpAxis=false) {enabled=false}
   &ScaleMeshSize(Scale Factor=&VectorFromFloat(Value=&FloatFromCurve(FloatCurve=DataInterface() {propertyValues {Curve {keys [<at 0.65 1>,<at 1 0>]}}}))) {enabled=false}
  ]
  rendererClasses [&NiagaraMeshRendererProperties]
  renderers [{rendererClass=&NiagaraMeshRendererProperties, rendererIndex=0}]
 }
 RollingSkull {
  emitterUpdateScript [
   &EmitterState()
   &SpawnBurst_Instantaneous(Spawn Count=3, Use Spawn Probability=false, UseLoopCountLimit=false)
  ]
  particleSpawnScript [
   &InitializeParticle(Color Mode=::NewEnumerator0, Lifetime=99999, UsePositionOffset=false, Write Material Random=true)
   &FN_SetSpawnGroup(Number of Spawn Groups=3)
  ]
  particleUpdateScript [
   &ParticleState()
   /SetVariables(Particles.New Matrix:NiagaraMatrix=/Niagara/DynamicInputs/Math/MakeTransform(
     Inverse:NiagaraBool=false
     Position:Vector3f=<0 -73.441 0>
     Local Offset:Vector3f=<0 0 100>
     Scale:Vector3f=<1 1 1>
     Rotation Interface:ENiagara_RotationDefinitionApproach=::NewEnumerator2
     Axis:Vector3f=?
     Quaternion:Quat4f=?
     Rotation Matrix:NiagaraMatrix=?
     X Vector:Vector3f=?
     Yaw:NiagaraFloat=0
     Angle:NiagaraFloat=?
     Pitch:NiagaraFloat=&SelectFloatFromArray(Direct Array Index=$Particles.SpawnGroup, Float Selection Array=DataInterface() {propertyValues {FloatData [68.4975,30,-55.9005]}})
     Y Vector:Vector3f=?
     Roll:NiagaraFloat=&Multiply_Float(
      A=&Add_Float(
       A=$Emitter.Age
       B=&SelectFloatFromArray(Direct Array Index=$Particles.SpawnGroup, Float Selection Array=DataInterface() {propertyValues {FloatData [-100,10,20]}})
      )
      B=-100
     )
     Z Vector:Vector3f=?
    )) {isSetParameters=true}
   /SetVariables(Particles.Position=/Niagara/DynamicInputs/Math/TransformPositionByMatrix(Position:Vector3f=<0 0 0>, TransformationMatrix:NiagaraMatrix=$Particles.New Matrix)) {isSetParameters=true}
   /SetVariables(Particles.MeshOrientation=&FindQuatBetweenTwoVectors(
     From Vector=<1 0 0>
     To Vector=&RotateVectorByQuaternion(
      Vector To Rotate=&NormalizeVector(
       Vector To Normalize=&SelectVectorFromArray(
        Direct Array Index=$Particles.SpawnGroup
        Vector Selection Array=DataInterface() {propertyValues {FloatData [<0.4 -1 -0.1>, <0 -1 0.3>, <-0.4 -1 -0.1>]}}
       )
       Safe Normalize=true
       Safe Fallback Threshold=0
       Safe Fallback Vector=<0 0 0.001>
      )
      Quaternion=&MultiplyQuaternion(
       Quaternion A=&MakeQuatFromAxisAngle(
        Axis=<0 0 1>
        AngleInDegrees=&Sine(
         Normalized Angle=&Add_Float(A=$Emitter.Age, B=&RandomRangeFloat(Minimum=0, Maximum=100))
         Period=2
         Scale=20
         Bias=0
        )
       )
       Quaternion B=&MakeQuatFromAxisAngle(
        Axis=<1 0 0>
        AngleInDegrees=&Sine(
         Normalized Angle=&Add_Float(A=$Emitter.Age, B=&RandomRangeFloat(Minimum=0, Maximum=100))
         Period=3
         Scale=20
         Bias=0
        )
       )
      )
     )
    )) {isSetParameters=true}
   &ScaleColor(
    Scale Alpha=&Sine(
     Normalized Angle=&Add_Float(A=$Emitter.Age, B=&RandomRangeFloat(Minimum=0, Maximum=50))
     Period=3
     Scale=0.7
     Bias=0.8
    )
    ScaleA=true
    ScaleRGB=true
   )
   /SetVariables(Particles.Mass:NiagaraFloat=&MakeFloatFromInt(Int=$Particles.SpawnGroup)) {isSetParameters=true}
   &ScaleColor(Scale Alpha=3, ScaleA=true, ScaleRGB=true) {module=ScaleColor001}
   &DynamicMaterialParameters(
    Index 0 Param 1=&Sine(
     Normalized Angle=&Add_Float(A=$Emitter.Age, B=&RandomRangeFloat(Minimum=0, Maximum=50))
     Period=1
     Scale=0.25
     Bias=0.75
    )
    Index 0 Param 2=&Subtract_Float(
     A=&MakeFloatFromVector(Vector=Linked() {linkedVariable {name=Engine.Owner.Position, type=NiagaraPosition}}, Channel=::NewEnumerator2)
     B=&MakeCustomFloatFromBool(Bool=$User.IsFrontEnd, True Float=40, False Float=15)
    )
    Index 0 Param 3=1
    Module.Param0WriteEnabled=true
    Module.Param1WriteEnabled=true
    Module.Param2WriteEnabled=true
    Module.Param3WriteEnabled=true
   )
   &ScaleColor(
    Scale Alpha=&FNTimeOfDayScalar(Night Scale=1, Morning Scale=0.5, Evening Scale=0.5)
    ScaleA=true
    ScaleRGB=true
   ) {module=ScaleColor002}
   &ScaleColor(
    Scale Alpha=&MakeCustomFloatFromBool(Bool=$User.IsFrontEnd, True Float=2.5, False Float=1)
    ScaleA=true
    ScaleRGB=true
   ) {module=ScaleColor003}
  ]
  rendererClasses [&NiagaraMeshRendererProperties]
  renderers [{rendererClass=&NiagaraMeshRendererProperties, rendererIndex=0}]
 }
}
summary {
 systemName=FX_Thrive_SpiritRoll
 emitters [
  {emitterName=RollingSkull, rendererClasses [&NiagaraMeshRendererProperties]}
  {emitterName=FollowingRibbon, rendererClasses [&NiagaraRibbonRendererProperties]}
  {emitterName=Empty002, rendererClasses [&NiagaraMeshRendererProperties]}
  {emitterName=Empty003, rendererClasses [&NiagaraMeshRendererProperties]}
  {emitterName=Empty004, rendererClasses [&NiagaraMeshRendererProperties]}
  {bEnabled=false, emitterName=MeshOrientation, rendererClasses [&NiagaraMeshRendererProperties]}
 ]
 userVariables [
  User.Deployed {type=NiagaraInt32}
  User.IsFrontEnd {type=NiagaraBool}
  User.SkullsOffset {type=Vector3f, description="", defaultValue {struct=Vector3f, value=<0 -20 15>}}
  User.StaticSkullsColor {type=LinearColor, description="", defaultValue {struct=LinearColor, value=<rgba 1 1 1 3>}}
 ]
}
