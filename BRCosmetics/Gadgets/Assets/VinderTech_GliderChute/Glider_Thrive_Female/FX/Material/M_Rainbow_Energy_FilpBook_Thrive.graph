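# Translucent unlit Niagara sprite material for rainbow energy flipbook effects in the Thrive female VinderTech glider's FX.
# Panning noise distorts the UVs of a smoke flipbook whose frame is driven by the Sub_shape dynamic parameter.
# Three offset samples form RGB fringes, oversaturated through negative desaturation and tinted by particle color.
# An unoffset sample supplies opacity, with depth fade skipped on forward and mobile paths.
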Material /BRCosmetics/Gadgets/Assets/VinderTech_GliderChute/Glider_Thrive_Female/FX/Material/M_Rainbow_Energy_FilpBook_Thrive
graph {
 properties {
  bUseTranslucencyVertexFog=false
  bUsedWithNiagaraSprites=true
  blendMode=BLEND_Translucent
  pixelDepthOffsetMode=PDOM_Legacy
  shadingModel=MSM_Unlit
  usedShadingModels=Unlit
 }
 nodes [
  n1 = ParticleColor()
  n2 = DynamicParameter() {defaultValue=<rgba 0 1 1 1>, paramNames [Sub_shape,Param2,Param3,Param4]}
  n3 = ScalarParameter() {defaultValue=2, parameterName=Sub_Number}
  n4 = TextureObjectParameter() {parameterName=Tex, texture=&/T_InkySmoke_Flipbook_blur}
  n5 = TextureCoordinate() {uTiling=0.8, vTiling=0.8}
  n6 = ComponentMask(None=VectorParameter() {group=Noise, parameterName=Noise_Scale, defaultValue=<rgba 2 2 0 0>}.RGB) {g=true, r=true}
  n7 = VectorParameter() {group=Noise, parameterName=Noise_Speed, defaultValue=<rgba 0.5 0.5 0 0>}
  n8 = TextureSampleParameter2D(UVs=Panner(Coordinate=Multiply(A=@n5, B=@n6), Speed=ComponentMask(None=@n7.RGB) {g=true, r=true})) {group=Noise, parameterName=Noise, samplerType=SAMPLERTYPE_Grayscale, texture=&LowResBlurredNoise}
  n9 = Multiply(A=@n8.R, B=ScalarParameter() {defaultValue=0.1, parameterName=Noise_Multiply})
  n10 = ComponentMask(None=VectorParameter() {group=Emissive, parameterName=UV_Offset, defaultValue=<rgba 0.04 0.04 0 1>}.RGB) {g=true, r=true}
  n11 = Add(A=Add(A=TextureCoordinate(), B=@n9), B=@n10)
  n12 = DynamicParameter() {defaultValue=<rgba 0 1 1 1>, paramNames [Sub_shape,Param2,Param3,Param4]}
  n13 = ScalarParameter() {defaultValue=2, parameterName=Sub_Number}
  n14 = TextureObjectParameter() {parameterName=Tex, texture=&T_InkySmoke_Flipbook_blur}
  n15 = Add(A=TextureCoordinate(), B=@n9)
  n16 = DynamicParameter() {defaultValue=<rgba 0 1 1 1>, paramNames [Sub_shape,Param2,Param3,Param4]}
  n17 = ScalarParameter() {defaultValue=2, parameterName=Sub_Number}
  n18 = TextureObjectParameter() {parameterName=Tex, texture=&T_InkySmoke_Flipbook_blur}
  n19 = Subtract(A=Add(A=TextureCoordinate(), B=@n9), B=@n10)
  n20 = Desaturation(
   None=MaterialFunctionCall(
    X=MaterialFunctionCall("Animation  Phase (0-1)"=@n2.Sub_shape, Number of Rows=@n3, Number of Columns=@n3, Texture=@n4, UVs=@n11) {materialFunction=&FlipBook}.Result
    Y=MaterialFunctionCall("Animation  Phase (0-1)"=@n12.Sub_shape, Number of Rows=@n13, Number of Columns=@n13, Texture=@n14, UVs=@n15) {materialFunction=&FlipBook}.Result
    Z=MaterialFunctionCall("Animation  Phase (0-1)"=@n16.Sub_shape, Number of Rows=@n17, Number of Columns=@n17, Texture=@n18, UVs=@n19) {materialFunction=&FlipBook}.Result
   ) {materialFunction=&MakeFloat3}.Result
   Fraction=ScalarParameter() {defaultValue=-2, group=Emissive, parameterName=Desaturation}
  )
  n21 = VectorParameter() {parameterName=Refine, channelNames {a="", b=Emissive_Power, g=Alpha_Multiply, r=Alpha_Power}, defaultValue=<rgba 1 1 0.8 0>}
  MaterialFunctionCall(Emissive=Multiply(A=@n1.RGB, B=Power(Base=@n20, Exp=@n21.Emissive_Power))) {materialFunction=&MF_FX_EmissiveAdjustments}
  n22 = ScalarParameter() {defaultValue=2, parameterName=Sub_Number}
  n23 = TextureObjectParameter() {parameterName=Tex, texture=&T_InkySmoke_Flipbook_blur}
  n24 = MaterialFunctionCall(
   "Animation  Phase (0-1)"=DynamicParameter() {defaultValue=<rgba 0 1 1 1>, paramNames [Sub_shape,Param2,Param3,Param4]}.Sub_shape
   Number of Rows=@n22
   Number of Columns=@n22
   Texture=@n23
   UVs=Add(A=TextureCoordinate(), B=@n9)
  ) {materialFunction=&FlipBook}
  n25 = Multiply(A=@n1.A, B=Multiply(A=Power(Base=@n24.Result, Exp=@n21.Alpha_Power), B=@n21.Alpha_Multiply))
  n26 = DepthFade(Opacity=@n25, FadeDistance=ScalarParameter() {defaultValue=20, parameterName=DepthFade})
  ShadingPathSwitch(Default=@n26, Deferred=@n26, Forward=@n25, Mobile=@n25)
 ]
}
