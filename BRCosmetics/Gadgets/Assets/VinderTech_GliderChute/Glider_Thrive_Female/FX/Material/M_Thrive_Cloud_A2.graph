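# Translucent lit material for the Thrive glider's cloud effect on Niagara mesh particles.
# Two panning bubble textures drive vertex displacement, detail normals and hue-shifted shadow colors, lit by sky or fixed light direction with a backlit fresnel rim.
# Particle dynamic parameters supply scale, timing, growth toward particle center and two foot positions whose sphere masks dent the cloud.
# A noise-warped ribbon texture tints it; depth fade and distance-based backface culling shape opacity.
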Material /BRCosmetics/Gadgets/Assets/VinderTech_GliderChute/Glider_Thrive_Female/FX/Material/M_Thrive_Cloud_A2
graph {
 properties {
  allowTranslucentCustomDepthWrites=true
  bUsedWithNiagaraMeshParticles=true
  blendMode=BLEND_Translucent
  opacityMaskClipValue=0.01
  pixelDepthOffsetMode=PDOM_Legacy
  usedShadingModels=Default Lit
 }
 nodes [
  Constant()
  Constant()
  n1 = ScalarParameter() {group=Misc, parameterName=TranslucentFresnelContrast}
  n2 = MaterialFunctionCall(In=Fresnel() {baseReflectFraction=0}, Contrast=@n1) {materialFunction=&CheapContrast}
  n3 = MaterialFunctionCall(In=Fresnel() {baseReflectFraction=0, exponent=7}, Contrast=@n1) {materialFunction=&CheapContrast}
  n4 = Transform(None=Constant3Vector() {constant=<rgba 0 0 1 0>}) {transformSourceType=TRANSFORMSOURCE_Local}
  n5 = VertexNormalWS()
  n6 = Normalize(VectorInput=Subtract(A=ObjectPositionWS(), B=CameraPositionWS()))
  n7 = Multiply(A=Add(A=DotProduct(A=@n4, B=@n6)) {constB=0.05}) {constB=-1}
  n8 = Add(A=Multiply(A=Abs(None=@n7)) {constB=-1}) {constB=0.15}
  n9 = Multiply(A=Saturate(None=DotProduct(A=@n4, B=@n5)), B=LinearInterpolate(Alpha=Saturate(None=@n8)) {constA=2, constB=14}) {constB=2}
  LinearInterpolate(
   A=LinearInterpolate(Alpha=OneMinus(None=@n2.Result))
   B=LinearInterpolate(Alpha=OneMinus(None=@n3.Result))
   Alpha=Power(Base=Saturate(None=@n9)) {constExponent=1.5}
  )
  n10 = DynamicParameter() {parameterIndex=1, paramNames [FeetPosX,FeetPosY,FeetPosZ,NoiseUVSCale]}
  n11 = AppendVector(A=AppendVector(A=@n10.FeetPosX, B=@n10.FeetPosX), B=@n10.FeetPosZ)
  n12 = WorldPosition() {worldPositionShaderOffset=WPT_ExcludeAllShaderOffsets}
  n13 = VectorParameter() {group=WPO_FeetHole, parameterName=FeetHole_SizeAndHardness, defaultValue=<rgba 30 0.5 0 0>}
  n14 = SphereMask(A=@n11, B=@n12.XYZ, Radius=@n13.R, Hardness=@n13.R)
  n15 = DynamicParameter() {parameterIndex=2, paramNames [Feet2PosX,Feet2PosY,Feet2PosZ,Grow]}
  n16 = AppendVector(A=AppendVector(A=@n15.Feet2PosX, B=@n15.Feet2PosX), B=@n15.Feet2PosZ)
  n17 = SphereMask(A=@n16, B=@n12.XYZ, Radius=@n13.R, Hardness=@n13.R)
  n18 = DynamicParameter() {parameterIndex=2, paramNames [Feet2PosX,Feet2PosY,Feet2PosZ,Grow]}
  n19 = OneMinus(None=@n18.Grow)
  n20 = LinearInterpolate(A=Max(A=@n14, B=@n17), Alpha=OneMinus(None=Power(Base=@n19))) {constB=0}
  n21 = VectorParameter() {group=Color, parameterName=CloudColor_BubbleBToColor_Remap, defaultValue=<rgba 0.5 1 1 0.5>}
  n22 = MaterialFunctionCall(UVs=TextureCoordinate(), Texture Scale=@n10.NoiseUVSCale) {materialFunction=&ScaleUVsByCenter}
  n23 = DynamicParameter() {parameterIndex=3, paramNames [FixRot, FE, "", Timer]}
  n24 = VectorParameter() {group=Bubble, parameterName="BubbleA_FlowSpeed&&Scale", defaultValue=<rgba 0.5 0 0.25 0.25>}
  n25 = AppendVector(A=@n24.R, B=@n24.R)
  n26 = VectorParameter() {group=Bubble, parameterName=BubbleA_Offset, defaultValue=<rgba 0 0.75 0 0>}
  n27 = Add(
   A=Multiply(A=Panner(Coordinate=@n22.UVs, Time=@n23.Timer, Speed=@n25), B=AppendVector(A=@n24.B, B=@n24.B))
   B=AppendVector(A=@n26.R, B=@n26.R)
  )
  n28 = TextureSample(UVs=@n27) {samplerType=SAMPLERTYPE_LinearGrayscale, texture=/Game/Gadgets/Assets/VinderTech_GliderChute/Glider_Thrive_Female/FX/Texture/T_CloudBubble}
  n29 = MaterialFunctionCall(X=@n28.R, Y=@n28.R, Z=@n28.R) {materialFunction=&MakeFloat3}
  n30 = ScalarParameter() {defaultValue=0.25, group=WPO_FeetHole, parameterName=FeetHole_LowestWPO}
  n31 = ScalarParameter() {defaultValue=0.05, group=WPO_FeetHole, parameterName=FeetHole_LowestWPOOffset}
  n32 = LinearInterpolate(A=@n29.Result, B=Add(A=Multiply(A=@n29.Result, B=@n30), B=@n31), Alpha=@n20)
  n33 = DotProduct(A=NamedRerouteUsage(), B=CameraVectorWS())
  n34 = Power(Base=Saturate(None=Multiply(A=@n33) {constB=-1})) {constExponent=10}
  n35 = Multiply(
   A=Fresnel() {baseReflectFraction=0, exponent=1}
   B=Multiply(
    A=MaterialFunctionCall(Input=@n32, Input Low=@n21.R, Input High=@n21.R, Target Low=@n21.R, Target High=@n21.R) {materialFunction=&RemapValueRange}.Result
    B=LinearInterpolate(Alpha=@n34) {constB=7}
   )
  ) {constB=10}
  n36 = NamedRerouteUsage()
  n37 = DotProduct(A=@n36, B=PixelNormalWS())
  n38 = Add(A=Multiply(A=@n37) {constB=-1}) {constB=0}
  n39 = ScalarParameter() {defaultValue=0.1, group=Color, parameterName=HueShift_A_Strength}
  n40 = TextureSample(UVs=@n27) {samplerType=SAMPLERTYPE_Normal, texture=/Game/Gadgets/Assets/VinderTech_GliderChute/Glider_Thrive_Female/FX/Texture/T_CloudBubble_N}
  n41 = ScalarParameter() {defaultValue=1.5, group=Lighting, parameterName=Lighting_DetailNormalStrength}
  n42 = Constant() {r=1}
  n43 = Normalize(VectorInput=Multiply(
    A=@n40.RGB
    B=MaterialFunctionCall(X=@n41, Y=@n41, Z=@n42) {materialFunction=&MakeFloat3}.Result
   ))
  n44 = VectorParameter() {group=Bubble, parameterName="BubbleB_FlowSpeed&&Scale", defaultValue=<rgba 0.85 0 0.3 0.3>}
  n45 = Multiply(A=Panner(Coordinate=@n22.UVs, Time=@n23.Timer, Speed=AppendVector(A=@n44.R, B=@n44.R)), B=AppendVector(A=@n44.B, B=@n44.B))
  n46 = TextureSample(UVs=@n45) {samplerType=SAMPLERTYPE_Normal, texture=/BRCosmetics/Gadgets/Assets/VinderTech_GliderChute/Glider_Thrive_Female/FX/Texture/T_CloudBubble_B_N}
  n47 = MaterialFunctionCall() {materialFunction=&LocalPosition}
  n48 = TransformPosition(None=@n11) {transformSourceType=TRANSFORMPOSSOURCE_World}
  n49 = VectorParameter() {group=WPO_FeetHole, parameterName=FeetHole_ScaleDirection, defaultValue=<rgba 1 1 0 1>}
  n50 = Normalize(VectorInput=Multiply(A=Subtract(A=@n47.Local Position, B=@n48), B=@n49.RGB))
  n51 = VectorParameter() {group=WPO_FeetHole, parameterName=FeetHole_NormalRot_LookAtVector}
  n52 = VectorParameter() {group=WPO_FeetHole, parameterName=FeetHole_NormalRot_RestingVector}
  n53 = Multiply(A=Subtract(A=@n47.Local Position, B=TransformPosition(None=@n16) {transformSourceType=TRANSFORMPOSSOURCE_World}), B=@n49.RGB)
  n54 = LinearInterpolate(
   A=LinearInterpolate(
    A=Constant3Vector() {constant=<rgba 0 0 1 0>}
    B=MaterialFunctionCall!(VectorToRotate=@n50, LookAtVector=@n51.RGB, RestingVector=@n52.RGB) {materialFunction=/Engine/Functions/Engine_MaterialFunctions01/Vectors/RotateVector}.Result
    Alpha=@n14
   )
   B=MaterialFunctionCall(VectorToRotate=Normalize(VectorInput=@n53), LookAtVector=@n51.RGB, RestingVector=@n52.RGB) {materialFunction=&RotateVector}.Result
   Alpha=@n17
  )
  n55 = ScalarParameter() {group=WPO_FeetHole, parameterName=FeetHole_NormalRot_Flatness}
  n56 = MaterialFunctionCall(
   Normal=LinearInterpolate(
    A=@n46.RGB
    B=MaterialFunctionCall(Normal=@n54, Flatness=@n55) {materialFunction=&FlattenNormal}.Result
    Alpha=@n20
   )
   Flatness=ScalarParameter() {defaultValue=0.1, group=Bubble, parameterName=BubbleB_NormalFlat}
  ) {materialFunction=&FlattenNormal}
  n57 = TextureSample(UVs=@n45) {
   samplerType=SAMPLERTYPE_LinearColor
   texture=/BRCosmetics/Gadgets/Assets/VinderTech_GliderChute/Glider_Thrive_Female/FX/Texture/T_CloudBubble_B
  }
  n58 = LinearInterpolate(A=@n57.RGB, B=Add(A=Multiply(A=@n57.RGB, B=@n30), B=@n31), Alpha=@n20)
  n59 = Saturate(None=Multiply(A=ComponentMask(None=@n58) {r=true}) {constB=2})
  n60 = PixelNormalWS()
  n61 = VertexNormalWS()
  n62 = Multiply(A=@n40.RGB, B=@n61)
  n63 = VectorParameter() {group=Misc, parameterName=BoundSpaceRotateAxis, defaultValue=<rgba 0 1 0 1>}
  n64 = ScalarParameter() {defaultValue=0.03, group=Lighting, parameterName=Lighting_Shadow_SpaceAngle}
  n65 = TransformPosition(None=WorldPosition() {worldPositionShaderOffset=WPT_ExcludeAllShaderOffsets}.XYZ) {transformSourceType=TRANSFORMPOSSOURCE_World}
  n66 = Add(A=RotateAboutAxis(NormalizedRotationAxis=@n63.RGB, RotationAngle=Add(A=@n23.FixRot, B=@n64), PivotPoint=Constant(), Position=@n65), B=@n65)
  n67 = MaterialFunctionCall() {materialFunction=&ObjectLocalBounds}
  n68 = MaterialFunctionCall(Float3=Divide(A=Subtract(A=@n66, B=@n67.Local Bounds Minimum), B=@n67.Local Bounds Size)) {materialFunction=&BreakOutFloat3Components}
  n69 = Add(A=@n68.B, B=ScalarParameter() {defaultValue=0.4, group=Lighting, parameterName=Lighting_Shadow_NormalRangeA_Offset}) {constB=0.4}
  n70 = Add(A=OneMinus(None=@n68.B), B=ScalarParameter() {defaultValue=0.6, group=Lighting, parameterName=Lighting_Shadow_NormalRangeB_Offset}) {constB=0.6}
  n71 = Min(
   A=Power(Base=Saturate(None=@n69), Exp=ScalarParameter() {defaultValue=5, group=Lighting, parameterName=Lighting_Shadow_NormalRangeA_Pow}) {constExponent=5}
   B=Power(Base=Saturate(None=@n70), Exp=ScalarParameter() {defaultValue=15, group=Lighting, parameterName=Lighting_Shadow_NormalRangeB_Pow}) {constExponent=15}
  )
  n72 = ScalarParameter() {defaultValue=5, group=Lighting, parameterName=Lighting_Shadow_NormalRange_Mul}
  n73 = LinearInterpolate(
   A=Transform(None=LinearInterpolate(A=@n43, B=@n56.Result, Alpha=@n59))
   B=Normalize(VectorInput=Add(A=@n60, B=@n62))
   Alpha=Saturate(None=Multiply(A=@n71, B=@n72) {constB=5})
  )
  n74 = Saturate(None=Add(A=DotProduct(A=@n36, B=@n73), B=ScalarParameter() {defaultValue=1, group=Lighting, parameterName=Lighting_Shadow_Detail_Offset}))
  n75 = ScalarParameter() {defaultValue=5.5, group=Lighting, parameterName=Lighting_Shadow_Detail_Power}
  n76 = Multiply(
   A=Add(A=@n37, B=ScalarParameter() {defaultValue=1.2, group=Lighting, parameterName=Lighting_Shadow_Smooth_Offset}) {constB=1.2}
   B=ScalarParameter() {defaultValue=0.5, group=Lighting, parameterName=Lighting_Shadow_Smooth_Mul}
  ) {constB=0.5}
  n77 = LinearInterpolate(
   A=VectorParameter() {group=Color, parameterName=CloudColor_SmoothShadowScaleColor, defaultValue=<rgba 0.2 0.2 0.2 1>}.RGB
   B=LinearInterpolate(
    A=VectorParameter() {group=Color, parameterName=CloudColor_DetailShadowScaleColor, defaultValue=<rgba 0.87 0.87 0.87 1>}.RGB
    B=VectorParameter() {group=Color, parameterName=CloudColor_BaseColor, defaultValue=<rgba 0 0.3 0.2 1>}.RGB
    Alpha=Power(Base=@n74, Exp=@n75) {constExponent=5.5}
   )
   Alpha=Power(Base=Saturate(None=@n76), Exp=ScalarParameter() {defaultValue=0.5, group=Lighting, parameterName=Lighting_Shadow_Smooth_Pow}) {constExponent=1}
  )
  n78 = Multiply(A=@n77, B=VectorParameter() {group=Color, parameterName=CloudColor_BubbleAToColor_Scale, defaultValue=<rgba 0 0 0 1>}.RGB)
  n79 = VectorParameter() {group=Color, parameterName=CloudColor_BubbleAToColor_Remap, defaultValue=<rgba 0 1 0.5 1>}
  n80 = MaterialFunctionCall(Input=@n29.Result, Input Low=@n79.R, Input High=@n79.R, Target Low=@n79.R, Target High=@n79.R) {materialFunction=&RemapValueRange}
  n81 = MaterialFunctionCall(
   Hue Shift Percentage=Multiply(A=OneMinus(None=@n32), B=ScalarParameter() {defaultValue=0.1, group=Color, parameterName=HueShift_B_Strength}) {constB=0.1}
   Texture=MaterialFunctionCall(Hue Shift Percentage=Multiply(A=@n38, B=@n39) {constB=0.1}, Texture=LinearInterpolate(A=@n78, B=@n77, Alpha=@n80.Result)) {materialFunction=&HueShift}.Result
  ) {materialFunction=&HueShift}
  n82 = TransformPosition(None=WorldPosition() {worldPositionShaderOffset=WPT_ExcludeAllShaderOffsets}.XYZ) {transformSourceType=TRANSFORMPOSSOURCE_World}
  n83 = Add(
   A=RotateAboutAxis(
    NormalizedRotationAxis=@n63.RGB
    RotationAngle=ScalarParameter() {defaultValue=0.03, group=BoundThings, parameterName=BoundThingsSpaceRotateAngle}
    PivotPoint=Constant()
    Position=@n82
   )
   B=@n82
  )
  n84 = MaterialFunctionCall() {materialFunction=&ObjectLocalBounds}
  n85 = MaterialFunctionCall(Float3=Divide(A=Subtract(A=@n83, B=@n84.Local Bounds Minimum), B=@n84.Local Bounds Size)) {materialFunction=&BreakOutFloat3Components}
  n86 = ScalarParameter() {group=BoundThings, parameterName=BoundHueShift_UVWOffset}
  n87 = MaterialFunctionCall(
   Hue Shift Percentage=Multiply(A=Add(A=@n85.G, B=@n86), B=ScalarParameter() {group=BoundThings, parameterName=BoundHueShift_UVWMul})
   Texture=VectorParameter() {group=Color, parameterName=TestRibbonColor, defaultValue=<rgba 0 0 0 1>}.RGB
  ) {materialFunction=&HueShift}
  n88 = Desaturation(None=@n81.Result, Fraction=Constant() {r=1})
  n89 = Add(A=@n85.B, B=ScalarParameter() {group=BoundThings, parameterName=BoundRibbonMask_UVWOffset})
  n90 = ScalarParameter() {group=BoundThings, parameterName=BoundRibbonMask_UVWMul}
  n91 = TextureCoordinate() {coordinateIndex=1}
  n92 = MaterialFunctionCall(Float2=@n91) {materialFunction=&BreakOutFloat2Components}
  n93 = VectorParameter() {group=RibbonTexture, parameterName=RibbonTex_NoiseStrecth_B, defaultValue=<rgba 1 -0.0115 0.2 0>}
  n94 = Multiply(A=OneMinus(None=@n92.R), B=@n93.R)
  n95 = LinearInterpolate(
   A=TextureCoordinate()
   B=TextureCoordinate() {coordinateIndex=1}
   Alpha=ScalarParameter() {group=RibbonTexture, parameterName=RibbonTex_NoiseUVChannel}
  )
  n96 = VectorParameter() {group=RibbonTexture, parameterName=RibbonTex_Noise1_UVScaleAndPanner}
  n97 = Multiply(A=@n95, B=AppendVector(A=@n96.R, B=@n96.R))
  n98 = VectorParameter() {group=RibbonTexture, parameterName=RibbonTex_Noise1_UVOffset}
  n99 = AppendVector(A=@n98.R, B=@n98.R)
  n100 = Time() {bOverride_Period=true, period=360}
  n101 = TextureSampleParameter2D(UVs=Panner(Coordinate=Add(A=@n97, B=@n99), Time=@n100, Speed=AppendVector(A=@n96.B, B=@n96.B)) {speedX=0.7795}) {
   group=RibbonTexture
   parameterName=RibbonTex_Noise1
   samplerType=SAMPLERTYPE_Normal
   texture=/BRCosmetics/Characters/Player/Male/Large/Bodies/M_LRG_Clash_V/Textures/Simple_Perlin_N
  }
  n102 = VectorParameter() {group=RibbonTexture, parameterName=RibbonTex_NoiseStrengthOffset}
  n103 = VectorParameter() {group=RibbonTexture, parameterName=RibbonTex_NoiseStrength}
  n104 = Multiply(A=Add(A=AppendVector(A=@n101.R, B=@n101.R), B=@n102.R) {constB=-0.5}, B=AppendVector(A=@n103.R, B=@n103.R)) {constB=0.264}
  n105 = VectorParameter() {group=RibbonTexture, parameterName=RibbonTex_Noise2_UVScaleAndPanner}
  n106 = Multiply(A=@n95, B=AppendVector(A=@n105.R, B=@n105.R))
  n107 = VectorParameter() {group=RibbonTexture, parameterName=RibbonTex_Noise2_UVScaleAndPanner}
  n108 = AppendVector(A=@n107.R, B=@n107.R)
  n109 = TextureSampleParameter2D(UVs=Panner(Coordinate=Add(A=@n106, B=@n108), Time=@n100, Speed=AppendVector(A=@n105.B, B=@n105.B)) {speedX=0.5}) {
   group=RibbonTexture
   parameterName=RibbonTex_Noise2
   samplerType=SAMPLERTYPE_LinearGrayscale
   texture=/Game/Global/Textures/Default/Noise/T_SwirlNoise_N_height
  }
  n110 = Multiply(A=Add(A=AppendVector(A=@n109.R, B=@n109.R), B=@n102.G) {constB=-0.5}, B=AppendVector(A=@n103.B, B=@n103.B)) {constB=0.264}
  n111 = Multiply(A=Saturate(None=Add(A=@n94, B=@n93.G) {constB=-0.0115}), B=Multiply(A=Add(A=@n104, B=@n110), B=@n93.B) {constB=0.2})
  n112 = OneMinus(None=@n92.R)
  n113 = VectorParameter() {group=RibbonTexture, parameterName=RibbonTex_NoiseStrecth_A, defaultValue=<rgba 1 -0.052 0.2 0>}
  n114 = Saturate(None=Add(A=Multiply(A=@n112, B=@n113.R), B=@n113.G) {constB=-0.052})
  n115 = AppendVector(A=@n113.B, B=@n113.B)
  n116 = TextureSampleParameter2D(UVs=Add(A=Add(A=@n91, B=@n111), B=Multiply(A=@n114, B=@n115) {constB=0.264})) {
   group=RibbonTexture
   parameterName=RibbonTex
   texture=/BRCosmetics/Gadgets/Assets/VinderTech_GliderChute/Glider_Thrive_Female/FX/Texture/T_CloudLine
  }
  n117 = Multiply(A=Saturate(None=Multiply(A=@n89, B=@n90)), B=@n116.R)
  n118 = ScalarParameter() {group=RibbonTexture, parameterName=RibbonTex_Contrast}
  n119 = Power(Base=OneMinus(None=DynamicParameter() {parameterIndex=2, paramNames [Feet2PosX,Feet2PosY,Feet2PosZ,Grow]}.Grow))
  n120 = ScalarParameter() {defaultValue=1, group=Misc, parameterName=DisableTexture}
  n121 = LinearInterpolate(
   A=@n81.Result
   B=LinearInterpolate(
    A=Multiply(A=@n81.Result, B=@n87.Result)
    B=Multiply(A=@n87.Result, B=@n88)
    Alpha=ScalarParameter() {group=RibbonTexture, parameterName=RibbonTexture_BlendMethod}
   )
   Alpha=Multiply(
    A=MaterialFunctionCall(In=@n117, Contrast=@n118) {materialFunction=&CheapContrast}.Result
    B=Multiply(A=@n119, B=@n120)
   )
  ) {constAlpha=0}
  n122 = Multiply(A=@n121, B=VectorParameter() {parameterName=ESFallbackFix, defaultValue=<rgba 0.7 0.7 0.7 1>}.RGB)
  n123 = ShadingPathSwitch(Default=@n121, Forward=@n122, Mobile=@n122)
  n124 = Multiply(A=LinearInterpolate(Alpha=@n20) {constA=1, constB=0.7}, B=Multiply(A=Multiply(A=@n123) {constB=0.5}, B=Add(A=@n35)))
  n125 = Constant()
  n126 = ScalarParameter() {group=Misc, parameterName=Debug}
  n127 = ParticleColor()
  MaterialFunctionCall(Emissive=Multiply(A=LinearInterpolate(A=@n124, B=@n125, Alpha=@n126), B=@n127.RGB)) {materialFunction=&MF_FX_EmissiveAdjustments}
  Multiply(A=@n127.RGB, B=LinearInterpolate(A=Multiply(A=@n123) {constB=0.85}, Alpha=@n126))
  n128 = Constant() {r=1}
  n129 = ScalarParameter() {defaultValue=10, group=Misc, parameterName=DepthFadeDistance}
  n130 = Constant() {r=1}
  n131 = MaterialFunctionCall!(ObjectOrParticlePosition) {materialFunction=/Game/Effects/Fort_Effects/Materials/Toon/CustomDepthOutLine/MF_BackfaceCulling}
  n132 = If(A=Distance(A=ParticlePositionWS(), B=CameraPositionWS()), "A > B"=@n130, "A == B"=@n131.Result, "A < B"=@n131.Result) {constB=10000}
  Multiply(
   A=Multiply(A=ShadingPathSwitch(Default=@n132, Forward=@n130, Mobile=@n130), B=LinearInterpolate(Alpha=@n20) {constA=1})
   B=Power(Base=DepthFade(Opacity=@n128, FadeDistance=@n129))
  )
  n133 = VectorParameter() {parameterName=FELightDirection, defaultValue=<rgba 0.5 -1 1 1>}
  NamedRerouteDeclaration(None=LinearInterpolate(
    A=SkyAtmosphereLightDirection()
    B=Normalize(VectorInput=@n133.RGB)
    Alpha=DynamicParameter() {parameterIndex=3, paramNames [FixRot, FE, "", Timer]}.FE
   )) {name=SkyLightDirect, nodeColor=<rgba 0.9922 0 1 1>}
  n134 = ParticlePositionWS()
  n135 = LinearInterpolate(
   A=@n134
   B=Add(A=@n134, B=VectorParameter() {group=Misc, parameterName=Grow_Center, defaultValue=<rgba 0 0 0 1>}.RGB)
   Alpha=ScalarParameter() {parameterName=DisableGrowCenter}
  )
  n136 = WorldPosition() {worldPositionShaderOffset=WPT_ExcludeAllShaderOffsets}
  n137 = Subtract(A=LinearInterpolate(A=@n136.XYZ, B=@n135, Alpha=@n18.Grow), B=@n136.XYZ)
  n138 = DynamicParameter() {paramNames [ScaleX,ScaleY,ScaleZ,ScaleWPOStr]}
  n139 = LinearInterpolate(Alpha=@n68.B) {constA=0.75}
  n140 = LinearInterpolate(
   A=ScalarParameter() {defaultValue=-0.3, group=Bubble, parameterName=Bubble_WPOOffset}
   B=ScalarParameter() {defaultValue=-0.3, group=Bubble, parameterName=Bubble_WPOOffset2}
   Alpha=@n18.Grow
  )
  n141 = Multiply(A=Add(A=@n58, B=@n140) {constB=-0.5}, B=ScalarParameter() {defaultValue=0.1, group=Bubble, parameterName=BubbleB_WPOStrength}) {constB=0.25}
  n142 = Add(A=@n32, B=@n140) {constB=-0.5}
  n143 = AppendVector(A=@n138.ScaleX, B=@n138.ScaleX)
  n144 = Multiply(
   A=Multiply(A=Add(A=@n141, B=@n142), B=@n61)
   B=Multiply(A=ScalarParameter() {defaultValue=400, group=Misc, parameterName=WPOGlobalScale}, B=AppendVector(A=@n143, B=@n138.ScaleX)) {constB=0.35}
  ) {constB=100}
  n145 = MaterialFunctionCall(Float2=TextureCoordinate()) {materialFunction=&BreakOutFloat2Components}
  n146 = Multiply(A=@n138.ScaleWPOStr, B=Multiply(A=Multiply(A=@n144, B=@n139), B=LinearInterpolate(Alpha=@n145.R) {constA=0.5}))
  n147 = Subtract(A=Add(A=Add(A=@n137, B=@n146), B=@n136.XYZ), B=@n135)
  n148 = LinearInterpolate(A=Constant3Vector() {constant=<rgba 1 1 1 0>}, B=Constant3Vector() {constant=<rgba 1 1 0.75 0>}, Alpha=@n18.Grow)
  Subtract(A=Add(A=@n135, B=Multiply(A=@n147, B=@n148)), B=WorldPosition() {worldPositionShaderOffset=WPT_ExcludeAllShaderOffsets}.XYZ)
 ]
}
