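# Translucent lit cloud material for Niagara mesh particles on the female Thrive glider.
# Panning bubble UVs and height terms build displacement along vertex normals, scaled by particle dynamic parameters.
# Two particle-driven foot positions make sphere masks that bend normals and lower displacement around the feet.
# Sky light direction drives stylized volume shading, hue shift and backlit Fresnel glow, with Fresnel edge, depth-fade and custom-depth terms for opacity.
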Material /BRCosmetics/Gadgets/Assets/VinderTech_GliderChute/Glider_Thrive_Female/FX/Material/M_Thrive_Cloud_SimpleA1
graph {
 properties {
  allowTranslucentCustomDepthWrites=true
  bUsedWithNiagaraMeshParticles=true
  blendMode=BLEND_Translucent
  opacityMaskClipValue=0.01
  pixelDepthOffsetMode=PDOM_Legacy
  usedShadingModels=Default Lit
 }
 nodes [
  n1 = TextureCoordinate()
  n2 = Time() {bOverride_Period=true, period=360}
  n3 = VectorParameter() {parameterName="Texture_BubbleA_FlowSpeed&&Scale", defaultValue=<rgba 0.5 0 0.25 0.25>}
  n4 = AppendVector(A=@n3.R, B=@n3.R)
  n5 = VectorParameter() {parameterName=Texture_BubbleA_Offset, defaultValue=<rgba 0 0.75 0 0>}
  Add(A=Multiply(A=Panner(Coordinate=@n1, Time=@n2, Speed=@n4), B=AppendVector(A=@n3.B, B=@n3.B)), B=AppendVector(A=@n5.R, B=@n5.R))
  Constant3Vector() {constant=<rgba 0 0 1 0>}
  Constant()
  Constant()
  n6 = Constant3Vector() {constant=<rgba 0 0 1 0>}
  n7 = MaterialFunctionCall() {materialFunction=&LocalPosition}
  n8 = DynamicParameter() {parameterIndex=1, paramNames [FeetPosX,FeetPosY,FeetPosZ,Param4]}
  n9 = AppendVector(A=AppendVector(A=@n8.FeetPosX, B=@n8.FeetPosX), B=@n8.FeetPosZ)
  n10 = Subtract(A=@n7.Local Position, B=TransformPosition(None=@n9) {transformSourceType=TRANSFORMPOSSOURCE_World})
  n11 = VectorParameter() {parameterName=ScaleDirection, defaultValue=<rgba 1 1 0 1>}
  n12 = VectorParameter() {parameterName="LookAtVector (V3)"}
  n13 = VectorParameter() {parameterName="RestingVector (V3)"}
  n14 = MaterialFunctionCall!(VectorToRotate=Normalize(VectorInput=Multiply(A=@n10, B=@n11.RGB)), LookAtVector=@n12.RGB, RestingVector=@n13.RGB) {materialFunction=/Engine/Functions/Engine_MaterialFunctions01/Vectors/RotateVector}
  n15 = WorldPosition() {worldPositionShaderOffset=WPT_ExcludeAllShaderOffsets}
  n16 = VectorParameter() {parameterName=feethole, defaultValue=<rgba 30 0.5 0 0>}
  n17 = SphereMask(A=@n9, B=@n15.XYZ, Radius=@n16.R, Hardness=@n16.R)
  n18 = DynamicParameter() {parameterIndex=2, paramNames [Feet2PosX,Feet2PosY,Feet2PosZ,Param4]}
  n19 = AppendVector(A=AppendVector(A=@n18.Feet2PosX, B=@n18.Feet2PosX), B=@n18.Feet2PosZ)
  n20 = TransformPosition(None=@n19) {transformSourceType=TRANSFORMPOSSOURCE_World}
  n21 = Normalize(VectorInput=Multiply(A=Subtract(A=@n7.Local Position, B=@n20), B=@n11.RGB))
  n22 = SphereMask(A=@n19, B=@n15.XYZ, Radius=@n16.R, Hardness=@n16.R)
  n23 = MaterialFunctionCall(
   Normal=LinearInterpolate(
    A=LinearInterpolate(A=@n6, B=@n14.Result, Alpha=@n17)
    B=MaterialFunctionCall(VectorToRotate=@n21, LookAtVector=@n12.RGB, RestingVector=@n13.RGB) {materialFunction=&RotateVector}.Result
    Alpha=@n22
   )
   Flatness=ScalarParameter() {parameterName="Flatness (S)"}
  ) {materialFunction=&FlattenNormal}
  n24 = Max(A=@n17, B=@n22)
  LinearInterpolate(B=@n23.Result, Alpha=@n24)
  LinearInterpolate(Alpha=@n24) {constA=1, constB=0.7}
  n25 = SkyAtmosphereLightDirection()
  n26 = DotProduct(A=@n25, B=PixelNormalWS())
  n27 = Multiply(A=@n26) {constB=-1}
  n28 = VectorParameter() {parameterName=Shadow_ScaleColor, defaultValue=<rgba 0.87 0.87 0.87 1>}
  n29 = VectorParameter() {parameterName=BaseColor, defaultValue=<rgba 0 0.3 0.2 1>}
  n30 = PixelNormalWS()
  n31 = DotProduct(A=@n25, B=@n30)
  n32 = ScalarParameter() {defaultValue=1, parameterName=Shadow_Offset}
  n33 = Power(Base=Saturate(None=Add(A=@n31, B=@n32)), Exp=ScalarParameter() {defaultValue=5.5, parameterName=Shadow_Power}) {constExponent=5.5}
  n34 = ScalarParameter() {defaultValue=1.2, parameterName=VolumeOffset}
  n35 = Saturate(None=Multiply(A=Add(A=@n26, B=@n34) {constB=1.2}, B=ScalarParameter() {defaultValue=0.5, parameterName=VolumeMul}) {constB=0.5})
  n36 = ScalarParameter() {defaultValue=0.5, parameterName=VolumePow}
  n37 = MaterialFunctionCall(
   Hue Shift Percentage=Multiply(A=Add(A=@n27) {constB=0}, B=ScalarParameter() {defaultValue=0.1, parameterName=Hueshift}) {constB=0.1}
   Texture=LinearInterpolate(
    A=VectorParameter() {parameterName=Volume_ScaleColor, defaultValue=<rgba 0.2 0.2 0.2 1>}.RGB
    B=LinearInterpolate(A=@n28.RGB, B=@n29.RGB, Alpha=@n33)
    Alpha=Power(Base=@n35, Exp=@n36) {constExponent=1}
   )
  ) {materialFunction=&HueShift}
  n38 = VectorParameter() {parameterName=TestRibbonColor, defaultValue=<rgba 0 0 0 1>}
  LinearInterpolate(
   A=@n37.Result
   B=Multiply(A=@n37.Result, B=MaterialFunctionCall(Texture=@n38.RGB) {materialFunction=&HueShift}.Result)
   Alpha=Multiply(B=ScalarParameter() {defaultValue=1, parameterName=DisableTexture})
  ) {constAlpha=0}
  n39 = ScalarParameter() {defaultValue=1.5, parameterName=Lighting_DetailNormalStrength}
  n40 = Constant() {r=1}
  n41 = Normalize(VectorInput=Multiply(B=MaterialFunctionCall(X=@n39, Y=@n39, Z=@n40) {materialFunction=&MakeFloat3}.Result))
  n42 = MaterialFunctionCall(Normal=LinearInterpolate(B=@n23.Result, Alpha=@n24), Flatness=ScalarParameter() {defaultValue=0.1, parameterName=Bubble2Flat}) {materialFunction=&FlattenNormal}
  n43 = ScalarParameter() {defaultValue=0.25, parameterName=LowWPO}
  n44 = ScalarParameter() {defaultValue=0.05, parameterName=LowWPOOffset}
  n45 = LinearInterpolate(B=Add(A=Multiply(B=@n43), B=@n44), Alpha=@n24)
  n46 = Saturate(None=Multiply(A=ComponentMask(None=@n45) {r=true}) {constB=2})
  n47 = VertexNormalWS()
  n48 = Multiply(B=@n47)
  n49 = TransformPosition(None=WorldPosition() {worldPositionShaderOffset=WPT_ExcludeAllShaderOffsets}.XYZ) {transformSourceType=TRANSFORMPOSSOURCE_World, transformType=TRANSFORMPOSSOURCE_Instance}
  n50 = Add(
   A=RotateAboutAxis(
    NormalizedRotationAxis=VectorParameter() {parameterName=Axis, defaultValue=<rgba 0 1 0 1>}.RGB
    RotationAngle=ScalarParameter() {defaultValue=0.03, parameterName=testa}
    PivotPoint=Constant()
    Position=@n49
   )
   B=@n49
  )
  n51 = MaterialFunctionCall() {materialFunction=&ObjectLocalBounds}
  n52 = MaterialFunctionCall(Float3=Divide(A=Subtract(A=@n50, B=@n51.Local Bounds Minimum), B=@n51.Local Bounds Size)) {materialFunction=&BreakOutFloat3Components}
  n53 = Add(A=@n52.B, B=ScalarParameter() {defaultValue=0.4, parameterName=ShadowNormalA_Offset}) {constB=0.4}
  n54 = Add(A=OneMinus(None=@n52.B), B=ScalarParameter() {defaultValue=0.6, parameterName=ShadowNormalB_Offset}) {constB=0.6}
  n55 = Min(
   A=Power(Base=Saturate(None=@n53), Exp=ScalarParameter() {defaultValue=5, parameterName=ShadowNormalA_Pow}) {constExponent=5}
   B=Power(Base=Saturate(None=@n54), Exp=ScalarParameter() {defaultValue=15, parameterName=ShadowNormalB_Pow}) {constExponent=15}
  )
  n56 = ScalarParameter() {defaultValue=5, parameterName=ShadowNormal_Mul}
  LinearInterpolate(
   A=Transform(None=LinearInterpolate(A=@n41, B=@n42.Result, Alpha=@n46))
   B=Normalize(VectorInput=Add(A=@n30, B=@n48))
   Alpha=Saturate(None=Multiply(A=@n55, B=@n56) {constB=5})
  )
  n57 = Multiply(
   A=Add(A=@n31, B=ScalarParameter() {defaultValue=1.2, parameterName=Volume2Offset}) {constB=1.2}
   B=ScalarParameter() {defaultValue=0.5, parameterName=Volume2Mul}
  ) {constB=0.5}
  LinearInterpolate(
   A=@n29.RGB
   B=VectorParameter() {parameterName=Volume2_Color, defaultValue=<rgba 0.2 0.2 0.2 1>}.RGB
   Alpha=Power(Base=Saturate(None=@n57), Exp=ScalarParameter() {defaultValue=0.5, parameterName=Volume2Pow}) {constExponent=1}
  )
  n58 = Fresnel() {baseReflectFraction=0, exponent=1}
  n59 = VectorParameter() {parameterName=remap2, defaultValue=<rgba 0.5 1 1 0.5>}
  n60 = MaterialFunctionCall() {materialFunction=&MakeFloat3}
  n61 = LinearInterpolate(A=@n60.Result, B=Add(A=Multiply(A=@n60.Result, B=@n43), B=@n44), Alpha=@n24)
  n62 = SkyAtmosphereLightDirection()
  n63 = CameraVectorWS()
  n64 = Saturate(None=Multiply(A=DotProduct(A=@n62, B=@n63)) {constB=-1})
  n65 = Multiply(
   A=MaterialFunctionCall(Input=@n61, Input Low=@n59.R, Input High=@n59.R, Target Low=@n59.R, Target High=@n59.R) {materialFunction=&RemapValueRange}.Result
   B=LinearInterpolate(Alpha=Power(Base=@n64) {constExponent=10}) {constB=7}
  )
  n66 = Multiply(A=Multiply(A=@n37.Result) {constB=0.5}, B=Add(A=Multiply(A=@n58, B=@n65) {constB=10}))
  n67 = ParticleColor()
  MaterialFunctionCall(Emissive=Multiply(A=ScalarParameter() {defaultValue=1, parameterName=TOD2}, B=Multiply(A=@n66, B=@n67.RGB))) {materialFunction=&MF_FX_EmissiveAdjustments}
  MaterialFunctionCall() {materialFunction=&BoundingBoxBased_0-1_UVW}
  MaterialFunctionCall(Hue Shift Percentage=Multiply(A=OneMinus(None=@n61)) {constB=0.1}, Texture=@n37.Result) {materialFunction=&HueShift}
  Multiply(A=ScalarParameter() {defaultValue=1, parameterName=TOD2}, B=Multiply(A=Multiply(A=@n37.Result) {constB=0.85}, B=@n67.RGB))
  n68 = VectorParameter() {defaultValue=<rgba 0.8 1.3 1 1>}
  Multiply(A=@n15.XYZ, B=@n68.RGB)
  Multiply(A=VectorParameter() {parameterName=FeetPosition, defaultValue=<rgba -240.1091 1852 2269 1>}.RGB, B=@n68.RGB)
  n69 = Constant() {r=1}
  n70 = ScalarParameter() {defaultValue=10, parameterName=sDF}
  n71 = Constant() {r=1}
  n72 = If(
   A=PixelDepth().R
   B=MaterialFunctionCall(Float3=SceneTexture() {sceneTextureId=PPI_CustomDepth}.Color) {materialFunction=&BreakOutFloat3Components}.R
   "A > B"=Constant()
   "A == B"=@n71
   "A < B"=@n71
  ) {equalsThreshold=0.1}
  n73 = Constant() {r=1}
  Multiply(A=Multiply(A=ShadingPathSwitch(Default=@n72, Forward=@n73, Mobile=@n73)), B=Power(Base=DepthFade(Opacity=@n69, FadeDistance=@n70)))
  n74 = VectorParameter() {parameterName="Texture_BubbleB_FlowSpeed&&Scale", defaultValue=<rgba 0.85 0 0.3 0.3>}
  Multiply(A=Panner(Coordinate=@n1, Time=@n2, Speed=AppendVector(A=@n74.R, B=@n74.R)), B=AppendVector(A=@n74.B, B=@n74.B))
  n75 = OneMinus(None=Fresnel(ExponentIn=ScalarParameter() {defaultValue=2, parameterName=OEdge}) {baseReflectFraction=0, exponent=2})
  n76 = ScalarParameter() {defaultValue=2, parameterName=ScaleEdge}
  Multiply(A=@n75, B=@n76)
  Multiply(A=Saturate(None=Power(Base=@n75, Exp=@n76)), B=@n67.A)
  n77 = LinearInterpolate(Alpha=@n52.B) {constA=0.75}
  n78 = ScalarParameter() {defaultValue=-0.3, parameterName=WPOOffset}
  n79 = Multiply(A=Add(A=@n45, B=@n78) {constB=-0.5}, B=ScalarParameter() {defaultValue=0.1, parameterName=Bubble2Strength}) {constB=0.25}
  n80 = Add(A=@n61, B=@n78) {constB=-0.5}
  n81 = DynamicParameter() {paramNames [ScaleX,ScaleY,ScaleZ,Param4]}
  n82 = AppendVector(A=@n81.ScaleX, B=@n81.ScaleX)
  n83 = Multiply(
   A=Multiply(A=Add(A=@n79, B=@n80), B=@n47)
   B=Multiply(A=ScalarParameter() {defaultValue=400, parameterName=T}, B=AppendVector(A=@n82, B=@n81.ScaleZ)) {constB=0.35}
  ) {constB=100}
  n84 = MaterialFunctionCall(Float2=TextureCoordinate()) {materialFunction=&BreakOutFloat2Components}
  Reroute(None=Multiply(A=Multiply(A=@n83, B=@n77), B=LinearInterpolate(Alpha=@n84.R) {constA=0.5}))
  Subtract(A=@n15.XYZ, B=@n19)
  Transform(None=@n68.RGB) {transformSourceType=TRANSFORMSOURCE_Local}
 ]
}
