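# Material for Niagara mesh particles in the Thrive female glider FX, rendering a golden tennis ball as lit.
# It turns the tennis ball normal map's strongest X or Y tilt into a boosted, saturated mask that blends white to gold.
# The blend is tinted by particle color, with an emissive branch scaled by the gold parameter's alpha and passed through a shared FX emissive adjustment function.
# Two scalar parameters, R and M, likely set roughness and metallic.
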Material /BRCosmetics/Gadgets/Assets/VinderTech_GliderChute/Glider_Thrive_Female/FX/Material/M_Thrive_GoldenTennis
graph {
 properties {bUsedWithNiagaraMeshParticles=true, pixelDepthOffsetMode=PDOM_Legacy, usedShadingModels=Default Lit}
 nodes [
  n1 = VectorParameter() {parameterName=G, defaultValue=<rgba 1 0.4833 0 0.2>}
  n2 = Constant3Vector() {constant=<rgba 1 0 0 0>}
  n3 = TextureSample() {samplerType=SAMPLERTYPE_Normal, texture=/Game/Environments/Apollo/Sets/DogHouse/Props/Textures/T_DogHouse_TennisBall_01_N}
  n4 = Constant3Vector() {constant=<rgba 0 1 0 0>}
  n5 = Max(A=Abs(None=DotProduct(A=@n2, B=@n3.RGB)), B=Abs(None=DotProduct(A=@n4, B=@n3.RGB)))
  n6 = LinearInterpolate(A=VectorParameter() {parameterName=E, defaultValue=<rgba 1 1 1 0.2>}.RGB, B=@n1.RGB, Alpha=Saturate(None=Multiply(A=@n5) {constB=6}))
  n7 = ParticleColor()
  MaterialFunctionCall(Emissive=Multiply(A=Multiply(A=@n6, B=@n1.A), B=@n7.RGB)) {materialFunction=&MF_FX_EmissiveAdjustments}
  Multiply(A=@n6, B=@n7.RGB)
  ScalarParameter() {defaultValue=0.35, parameterName=R}
  ScalarParameter() {defaultValue=1, parameterName=M}
 ]
}
