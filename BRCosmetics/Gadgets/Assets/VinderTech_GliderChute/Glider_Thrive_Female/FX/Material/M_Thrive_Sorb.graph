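# Additive Niagara mesh-particle material for the Thrive female glider's FX, drawing an animated, tinted energy glow with refraction.
# Two scrolling noise textures distort sheared UVs that sample a soft gradient three times at offset positions, forming a chromatic split.
# The refined channels are partly desaturated, tinted and scaled by particle alpha into emissive.
# Refraction follows gradient, vertical falloff, particle alpha and camera fade, with a constant on Forward and Mobile.
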Material /BRCosmetics/Gadgets/Assets/VinderTech_GliderChute/Glider_Thrive_Female/FX/Material/M_Thrive_Sorb
graph {
 properties {
  bUsedWithNiagaraMeshParticles=true
  blendMode=BLEND_Additive
  pixelDepthOffsetMode=PDOM_Legacy
  refractionMethod=RM_IndexOfRefraction
  usedShadingModels=Default Lit
 }
 nodes [
  n1 = VectorParameter() {parameterName=NoiseAValueOffsetAndMul, defaultValue=<rgba 0 -0.712 0 0.4>}
  n2 = AppendVector(A=@n1.B, B=@n1.B)
  n3 = MaterialFunctionCall(Float2=TextureCoordinate()) {materialFunction=&BreakOutFloat2Components}
  n4 = ScalarParameter() {defaultValue=0.35, parameterName=UVRotate}
  n5 = Add(A=TextureCoordinate(), B=AppendVector(A=Multiply(A=@n3.G, B=@n4), B=Constant()))
  n6 = VectorParameter() {parameterName=NoiseAUVScaleAndOffset, defaultValue=<rgba 2 0.5 0 0.5>}
  n7 = AppendVector(A=@n6.R, B=@n6.R)
  n8 = AppendVector(A=@n6.B, B=@n6.B)
  n9 = Multiply(A=Time(), B=ScalarParameter() {defaultValue=2.5, parameterName=TimeScale})
  n10 = TextureSample(UVs=Add(A=Multiply(A=@n5, B=@n7), B=Multiply(A=@n8, B=@n9))) {texture=&AtmosphericCloudNoise01}
  n11 = Add(A=AppendVector(A=@n10.R, B=@n10.R), B=AppendVector(A=@n1.R, B=@n1.R))
  n12 = VectorParameter() {parameterName=NoiseBValueOffsetAndMul, defaultValue=<rgba 0 -0.2079 0 1>}
  n13 = VectorParameter() {parameterName=NoiseBUVScaleAndOffset, defaultValue=<rgba 1 0.5 0 0.5>}
  n14 = AppendVector(A=@n13.R, B=@n13.R)
  n15 = AppendVector(A=@n13.B, B=@n13.B)
  n16 = TextureSample(UVs=Add(A=Multiply(A=@n5, B=@n14), B=Multiply(A=@n15, B=@n9))) {texture=/Game/Global/Textures/Default/Noise/3xPerlin_M}
  n17 = AppendVector(A=@n16.R, B=@n16.R)
  n18 = AppendVector(A=@n12.R, B=@n12.R)
  n19 = Add(A=Add(A=Multiply(A=@n2, B=@n11), B=@n5), B=Multiply(A=AppendVector(A=@n12.B, B=@n12.B), B=Add(A=@n17, B=@n18)))
  n20 = VectorParameter() {parameterName=UVOffset, defaultValue=<rgba 0 0.05 0 1>}
  n21 = AppendVector(A=@n20.R, B=@n20.R)
  n22 = TextureSample(UVs=Add(A=@n19, B=@n21)) {samplerSource=SSM_Clamp_WorldGroupSettings, texture=/Game/Global/Textures/Default/Masks_and_Gradients/T_Vert_Gradient_Soft}
  n23 = VectorParameter() {parameterName=Refine, defaultValue=<rgba 10 0.5 0.1 0.1>}
  n24 = Multiply(A=Power(Base=@n22.RGB, Exp=@n23.R) {constExponent=3}, B=@n23.G) {constB=2}
  n25 = Multiply(A=@n22.RGB, B=@n23.B)
  n26 = TextureSample(UVs=@n19) {samplerSource=SSM_Clamp_WorldGroupSettings, texture=&T_Vert_Gradient_Soft}
  n27 = Multiply(A=Power(Base=@n26.RGB, Exp=@n23.R) {constExponent=3}, B=@n23.G) {constB=2}
  n28 = Multiply(A=@n26.RGB, B=@n23.B)
  n29 = TextureSample(UVs=Add(A=@n19, B=Multiply(A=@n21) {constB=-1})) {samplerSource=SSM_Clamp_WorldGroupSettings, texture=&T_Vert_Gradient_Soft}
  n30 = Multiply(A=Power(Base=@n29.RGB, Exp=@n23.R) {constExponent=3}, B=@n23.G) {constB=2}
  n31 = Multiply(A=@n29.RGB, B=@n23.B)
  n32 = VectorParameter() {parameterName=Color, defaultValue=<rgba 0 0.85 0.8068 0.5>}
  n33 = Desaturation(
   None=MaterialFunctionCall(
    X=LinearInterpolate(A=@n24, B=@n25, Alpha=@n23.A)
    Y=LinearInterpolate(A=@n27, B=@n28, Alpha=@n23.A)
    Z=LinearInterpolate(A=@n30, B=@n31, Alpha=@n23.A)
   ) {materialFunction=&MakeFloat3}.Result
   Fraction=@n32.A
  )
  n34 = ParticleColor()
  MaterialFunctionCall(Emissive=Multiply(A=Multiply(A=@n33, B=@n32.RGB), B=@n34.A)) {materialFunction=&MF_FX_EmissiveAdjustments}
  n35 = OneMinus(None=MaterialFunctionCall(Float2=@n5) {materialFunction=&BreakOutFloat2Components}.G)
  n36 = Multiply(A=@n35, B=@n35)
  n37 = Multiply(A=Multiply(A=Multiply(A=@n36, B=@n26.R), B=ScalarParameter() {defaultValue=0.2, parameterName=RefractionStrength}), B=@n34.A)
  n38 = OneMinus(None=MaterialFunctionCall(Distance=ScalarParameter() {defaultValue=2048, group=Base, parameterName=Refraction Camera Fade Distance}) {materialFunction=&Near_Camera_Fade}.Result)
  n39 = Constant() {r=1}
  ShadingPathSwitch(Default=Add(A=Multiply(A=@n37, B=@n38)), Forward=@n39, Mobile=@n39)
 ]
}
