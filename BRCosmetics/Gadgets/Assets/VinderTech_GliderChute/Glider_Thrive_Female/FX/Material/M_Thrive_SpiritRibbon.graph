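# Translucent unlit material for Niagara ribbons, drawing the spirit ribbon effect of the female Thrive glider.
# Two panning, edge-masked noise layers distort the base-shape texture UVs, with a dynamic NoiseReducer parameter damping the distortion.
# Those UVs and vertex-alpha-randomized time drive a hue shift on the tint, multiplied by particle color for emissive.
# Opacity combines shape, U gradient and particle alpha with a custom-depth fade and occlusion test, the latter bypassed on forward and mobile.
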Material /BRCosmetics/Gadgets/Assets/VinderTech_GliderChute/Glider_Thrive_Female/FX/Material/M_Thrive_SpiritRibbon
graph {
 properties {bUsedWithNiagaraRibbons=true, blendMode=BLEND_Translucent, pixelDepthOffsetMode=PDOM_Legacy, shadingModel=MSM_Unlit, usedShadingModels=Unlit}
 nodes [
  Constant() {r=0.1}
  n1 = Time() {bOverride_Period=true, period=360}
  n2 = DynamicParameter() {parameterIndex=3, defaultValue=<rgba 0 0 0 0>, paramNames [Timer,0,0,0]}
  LinearInterpolate(A=@n1, B=@n2.Timer, Alpha=Saturate(None=Ceil(None=@n2.Timer)))
  n3 = ParticleColor()
  n4 = MaterialFunctionCall(Float2=NamedRerouteUsage()) {materialFunction=&BreakOutFloat2Components}
  n5 = Add(
   A=Multiply(A=OneMinus(None=@n4.R), B=ScalarParameter() {group=BaseShape, parameterName=BaseShape_Stretch})
   B=ScalarParameter() {group=BaseShape, parameterName=BaseShape_StretchOffset}
  )
  n6 = Constant()
  n7 = DynamicParameter() {parameterIndex=2, defaultValue=<rgba 1 0 0 1>, paramNames [MeshPosZ,EnableFollow,NoiseReducer,Grow]}
  n8 = Saturate(None=@n7.NoiseReducer)
  n9 = VectorParameter() {group=Texture_NoiseB, parameterName=NoiseB_SP, channelNames {a=PannerV, b=PannerU, g=ScaleV, r=ScaleU}, defaultValue=<rgba 1 0.1 0.2 0>}
  n10 = Multiply(A=NamedRerouteUsage(), B=AppendVector(A=@n9.ScaleU, B=@n9.ScaleU))
  n11 = VectorParameter() {group=Texture_NoiseB, parameterName=NoiseB_A, channelNames {a="", b="", g=OffsetV, r=OffsetU}, defaultValue=<rgba 1 0.1 0.2 0>}
  n12 = AppendVector(A=@n11.OffsetU, B=@n11.OffsetU)
  n13 = NamedRerouteUsage()
  n14 = TextureSampleParameter2D(UVs=Panner(Coordinate=Add(A=@n10, B=@n12), Time=@n13, Speed=AppendVector(A=@n9.PannerU, B=@n9.PannerU)) {speedX=0.216}) {
   group=Texture_NoiseB
   parameterName=NoiseB_Texture
   samplerType=SAMPLERTYPE_LinearGrayscale
   texture=/Game/Gadgets/Assets/VinderTech_GliderChute/Glider_Thrive_Female/FX/Texture/T_AntiMaterialCharge_Dissolve_01
  }
  n15 = AppendVector(A=@n14.R, B=@n14.R)
  n16 = VectorParameter() {
   group=Texture_NoiseB
   parameterName=NoiseB_Property
   channelNames {a=NoiseStrength_V, b=NoiseStrength_U, g=NoiseStrengthOffset_V, r=NoiseStrengthOffset_U}
   defaultValue=<rgba 1 0.1 0.2 0>
  }
  n17 = AppendVector(A=@n16.NoiseStrengthOffset_U, B=@n16.NoiseStrengthOffset_U)
  n18 = MaterialFunctionCall(Float2=NamedRerouteUsage()) {materialFunction=&BreakOutFloat2Components}
  n19 = VectorParameter() {
   group=Texture_NoiseB
   parameterName=NoiseB_MaskInfo
   channelNames {a=Buttomoffset, b=ButtomHardness, g=TopOffset, r=TopHardness}
   defaultValue=<rgba 3 0 5 2.4>
  }
  n20 = Add(A=Multiply(A=OneMinus(None=@n18.R), B=@n19.TopHardness) {constB=15}, B=@n19.TopOffset)
  n21 = Add(A=Multiply(A=@n18.R, B=@n19.ButtomHardness) {constB=15}, B=@n19.Buttomoffset)
  n22 = Multiply(
   A=Multiply(A=Add(A=@n15, B=@n17) {constB=-0.5}, B=AppendVector(A=@n16.NoiseStrength_U, B=@n16.NoiseStrength_U))
   B=Multiply(A=Saturate(None=@n20), B=Saturate(None=@n21))
  )
  n23 = VectorParameter() {group=Texture_NoiseA, parameterName=NoiseA_SP, channelNames {a=PannerV, b=PannerU, g=ScaleV, r=ScaleU}, defaultValue=<rgba 1 0.1 0.2 0>}
  n24 = Multiply(A=NamedRerouteUsage(), B=AppendVector(A=@n23.ScaleU, B=@n23.ScaleU))
  n25 = VectorParameter() {group=Texture_NoiseA, parameterName=NoiseA_A, channelNames {a="", b="", g=OffsetV, r=OffsetU}, defaultValue=<rgba 1 0.1 0.2 0>}
  n26 = AppendVector(A=@n25.OffsetU, B=@n25.OffsetU)
  n27 = TextureSampleParameter2D(UVs=Panner(Coordinate=Add(A=@n24, B=@n26), Time=@n13, Speed=AppendVector(A=@n23.PannerU, B=@n23.PannerU)) {speedX=0.216}) {
   group=Texture_NoiseA
   parameterName=NoiseA_Texture
   samplerType=SAMPLERTYPE_LinearGrayscale
   texture=&T_AntiMaterialCharge_Dissolve_01
  }
  n28 = AppendVector(A=@n27.R, B=@n27.R)
  n29 = VectorParameter() {
   group=Texture_NoiseA
   parameterName=NoiseA_Property
   channelNames {a=NoiseStrength_V, b=NoiseStrength_U, g=NoiseStrengthOffset_V, r=NoiseStrengthOffset_U}
   defaultValue=<rgba 1 0.1 0.2 0>
  }
  n30 = AppendVector(A=@n29.NoiseStrengthOffset_U, B=@n29.NoiseStrengthOffset_U)
  n31 = MaterialFunctionCall(Float2=NamedRerouteUsage()) {materialFunction=&BreakOutFloat2Components}
  n32 = VectorParameter() {
   group=Texture_NoiseA
   parameterName=NoiseA_MaskInfo
   channelNames {a=Buttomoffset, b=ButtomHardness, g=TopOffset, r=TopHardness}
   defaultValue=<rgba 5 0 25 5>
  }
  n33 = Add(A=Multiply(A=OneMinus(None=@n31.R), B=@n32.TopHardness) {constB=15}, B=@n32.TopOffset)
  n34 = Add(A=Multiply(A=@n31.R, B=@n32.ButtomHardness) {constB=15}, B=@n32.Buttomoffset)
  n35 = Multiply(
   A=Multiply(A=Add(A=@n28, B=@n30) {constB=-0.5}, B=AppendVector(A=@n29.NoiseStrength_U, B=@n29.NoiseStrength_U))
   B=Multiply(A=Saturate(None=@n33), B=Saturate(None=@n34))
  )
  n36 = Add(A=Add(A=AppendVector(A=@n5, B=@n6), B=NamedRerouteUsage()), B=Multiply(A=OneMinus(None=@n8), B=Add(A=@n22, B=@n35)))
  n37 = Add(
   A=NamedRerouteUsage()
   B=Multiply(
    A=MaterialFunctionCall(XY=@n36) {materialFunction=&Swizzle}.YX
    B=ScalarParameter() {group=BaseColor, parameterName=HueShiftNoiseScale}
   )
  )
  n38 = Add(
   A=Multiply(
    A=MaterialFunctionCall(Float2=@n37) {materialFunction=&BreakOutFloat2Components}.R
    B=ScalarParameter() {group=BaseColor, parameterName=HueShiftStr}
   )
   B=Multiply(A=NamedRerouteUsage(), B=ScalarParameter() {group=BaseColor, parameterName=HueShiftMoveSpeed})
  )
  n39 = VectorParameter() {parameterName=CCCCC, defaultValue=<rgba 1 1 1 1>}
  MaterialFunctionCall(Emissive=Multiply(
    A=@n3.RGB
    B=MaterialFunctionCall(Hue Shift Percentage=@n38, Texture=@n39.RGB) {materialFunction=&HueShift}.Result
   )) {materialFunction=&MF_FX_EmissiveAdjustments}
  n40 = MaterialFunctionCall(Float3=SceneTexture() {sceneTextureId=PPI_CustomDepth}.Color) {materialFunction=&BreakOutFloat3Components}
  n41 = PixelDepth()
  n42 = ComponentMask(None=@n41.R) {r=true}
  n43 = Saturate(None=Divide(A=Subtract(A=@n40.R, B=@n42), B=ScalarParameter() {defaultValue=20, parameterName=CustomDepthFade}) {constB=50})
  n44 = Constant() {r=1}
  n45 = Constant() {r=1}
  n46 = ShadingPathSwitch(Default=If(A=@n41.R, B=@n40.R, "A > B"=Constant(), "A == B"=@n44, "A < B"=@n44) {equalsThreshold=0.1}, Forward=@n45, Mobile=@n45)
  n47 = VectorParameter() {group=BaseShape, parameterName=BaseShape_RibbonScale, defaultValue=<rgba 1 1 0 1>}
  n48 = TextureSampleParameter2D(UVs=MaterialFunctionCall(UVs=@n36, Texture Scale=AppendVector(A=@n47.R, B=@n47.R)) {materialFunction=&ScaleUVsByCenter}.UVs) {
   group=BaseShape
   parameterName=BaseShapeTexture
   samplerSource=SSM_Clamp_WorldGroupSettings
   samplerType=SAMPLERTYPE_Grayscale
   texture=/Game/Gadgets/Assets/VinderTech_GliderChute/Glider_Thrive_Female/FX/Texture/T_RibbonBaseShape
  }
  n49 = Multiply(
   A=@n39.A
   B=MaterialFunctionCall(Float2=TextureCoordinate()) {materialFunction=&BreakOutFloat2Components}.R
  )
  Multiply(A=@n3.A, B=Multiply(A=Multiply(A=@n43, B=@n46), B=Multiply(A=@n48.R, B=@n49)))
  Multiply(A=@n48.R, B=@n39.RGB)
  Multiply(B=@n47.G)
  n50 = MaterialFunctionCall(Float2=TextureCoordinate()) {materialFunction=&BreakOutFloat2Components}
  n51 = OneMinus(None=@n50.R)
  n52 = ScalarParameter() {group=UVSettings, parameterName=FlipU}
  n53 = OneMinus(None=@n50.G)
  n54 = ScalarParameter() {group=UVSettings, parameterName=FlipV}
  NamedRerouteDeclaration(None=MaterialFunctionCall(X=LinearInterpolate(A=@n50.R, B=@n51, Alpha=@n52), Y=LinearInterpolate(A=@n50.G, B=@n53, Alpha=@n54)) {materialFunction=&MakeFloat2}.Result) {name=UV, nodeColor=<rgba 0 1 0.2266 1>}
  n55 = VertexColor()
  n56 = ScalarParameter() {defaultValue=10, group=Misc, parameterName=RandomScale}
  NamedRerouteDeclaration(None=Add(A=@n1, B=Multiply(A=@n55.A, B=@n56))) {name=TimeRandom, nodeColor=<rgba 0 0.3203 1 1>}
  TextureCoordinate()
 ]
}
